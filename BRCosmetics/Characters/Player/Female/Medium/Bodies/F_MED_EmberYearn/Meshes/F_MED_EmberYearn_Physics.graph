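# Physics asset for the EmberYearn medium female player body cosmetic, covering the upper spine and dynamic jacket bones.
# Capsules on the back, front, left and right jacket bones sit alongside larger capsules on spine_04 and spine_05.
# Each jacket bone is constrained to spine_04 with narrow swing limits and zero twist limit.
# The right jacket bone has a second, reversed constraint from spine_04 with swing and twist set free.
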PhysicsAsset /BRCosmetics/Characters/Player/Female/Medium/Bodies/F_MED_EmberYearn/Meshes/F_MED_EmberYearn_Physics
bodies {
 dyn_jacket_01_b [{
   shapeType=Capsule
   length=6.027
   radius=4.5967
   shapeName=dyn_jacket_01_b_capsule
   center=<0.3213 0.0239 2.1463>
   rotation=<rot 0 0.1556 -89.8085>
  }]
 dyn_jacket_01_f [{
   shapeType=Capsule
   length=4.5698
   radius=2.9746
   shapeName=dyn_jacket_01_f_capsule
   center=<4.4879 0.0903 1.9293>
   rotation=<rot 0 -179.6366 -89.6476>
  }]
 dyn_jacket_01_l [{shapeType=Capsule, length=6.4415, radius=2.8588, shapeName=dyn_jacket_01_l_capsule, center=<1.5702 0.0574 1.2553>, rotation=<rot 0 180 -90>}]
 dyn_jacket_01_r [{shapeType=Capsule, length=6.0577, radius=3.2426, shapeName=dyn_jacket_01_r_capsule, center=<3.5 0 2.1111>, rotation=<rot 0 180 90>}]
 spine_04 [
  {shapeType=Capsule, length=7.4934, radius=7.3563, center=<-1.0222 -0.0323 0.4573>, rotation=<rot 0 -90 0>}
  {shapeType=Capsule, length=3.9212, radius=6.9531, shapeName=spine_04_capsule, center=<-5.3939 5.0061 0>, rotation=<rot 0 6.586 0>}
 ]
 spine_05 [{shapeType=Capsule, length=14.5695, radius=10.7552, center=<6.9165 -0.238 0.0143>, rotation=<rot -0.2751 -90 -0.0488>}]
}
constraints [
 {bone1Name=dyn_jacket_01_b, bone2Name=spine_04, swing1LimitDegrees=15, swing2LimitDegrees=30, twistLimitDegrees=0}
 {bone1Name=dyn_jacket_01_f, bone2Name=spine_04, swing1LimitDegrees=15, swing2LimitDegrees=30, twistLimitDegrees=0}
 {
  bone1Name=spine_04
  bone2Name=dyn_jacket_01_r
  swing1LimitDegrees=45
  swing1Motion=Free
  swing2LimitDegrees=45
  swing2Motion=Free
  twistLimitDegrees=45
  twistMotion=Free
 }
 {bone1Name=dyn_jacket_01_r, bone2Name=spine_04, swing1LimitDegrees=15, swing2LimitDegrees=30, twistLimitDegrees=0}
 {bone1Name=dyn_jacket_01_l, bone2Name=spine_04, swing1LimitDegrees=15, swing2LimitDegrees=30, twistLimitDegrees=0}
]
