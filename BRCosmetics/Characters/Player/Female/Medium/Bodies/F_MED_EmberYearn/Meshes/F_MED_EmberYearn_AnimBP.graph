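# Animation Blueprint for the F_MED_EmberYearn female body cosmetic, adding secondary motion to a pose copied from the attached parent mesh.
# The pose runs through a pre Control Rig, pelvis-space RigidBody simulation, a weighted neck/spine constraint and spring AnimDynamics for the zipper, then a post Control Rig.
# Each update computes gravity and RigidBody settings from a parameter asset, and converts parachute and tactical sprint flags to floats for the post rig.
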AnimBlueprint /BRCosmetics/Characters/Player/Female/Medium/Bodies/F_MED_EmberYearn/Meshes/F_MED_EmberYearn_AnimBP
target_skeleton=&/F_MED_EmberYearn_Skeleton
graphs [
 AnimGraph {nodes [
   n1 = Control Rig(Source=Copy Pose From Mesh() {bCopyCurves=true, bUseAttachedParent=true}) {controlRigAssetReference {blueprintRigClass=&/F_MED_EmberYearn_Pre_CtrlRig.F_MED_EmberYearn_Pre_CtrlRig_C}}
   n2 = RigidBody!(
    ComponentPose=Local To Component(LocalPose=@n1)
    Alpha
    OverrideWorldGravity:Vector="0,0,0"
    ExternalForce
    SimSpaceSettings:Sim Space Settings Structure=(WorldAlpha=0,VelocityScaleZ=1,DampingAlpha=1,MaxLinearVelocity=10000,MaxAngularVelocity=10000,MaxLinearAcceleration=10000,MaxAngularAcceleration=10000,ExternalLinearDragV=(X=0,Y=0,Z=0),ExternalLinearVelocity=(X=0,Y=0,Z=0),ExternalAngularVelocity=(X=0,Y=0,Z=0))
   ) -> (Pose) {
    bOverrideWorldGravity=true
    overridePhysicsAsset=&/F_MED_EmberYearn_Physics
    simulationSpace=BaseBoneSpace
    baseBoneRef {boneName=pelvis}
    componentAppliedLinearAccClamp=<1000 1000 1000>
   }
   n3 = "Constraint - base_zipper_01 (PPPPP)"!(
    ComponentPose=@n2
    Alpha=1
    ConstraintWeights_0=0.3658
    ConstraintWeights_1:float=0.3589
    ConstraintWeights_2:float=0.2008
    ConstraintWeights_3:float=0.0724
    ConstraintWeights_4:float=0.0021
   ) -> (Pose) {
    boneToModify {boneName=base_zipper_01}
    constraintSetup [
     {targetBone {boneName=neck_02}}
     {targetBone {boneName=spine_05}}
     {targetBone {boneName=neck_01}}
     {targetBone {boneName=spine_04}}
     {targetBone {boneName=spine_03}}
    ]
    constraintWeights [1,1,1,1,1]
   }
   n4 = Anim Dynamics - Bone: dyn_zipper_02(ComponentPose=@n3, Alpha=1, ExternalForce="0,0,0") {
    angularDampingOverride=0.9999
    angularSpringConstant=10000
    bAngularSpring=true
    bOverrideAngularDamping=true
    bUseGravityOverride=true
    gravityScale=0.35
    simulationSpace=RootRelative
    boundBone {boneName=dyn_zipper_02}
    componentLinearAccScale=<1 1 1>
    componentLinearVelScale=<1 1 1>
    gravityOverride {z=-980}
    physicsBodyDefinitions [{
      boundBone {boneName=dyn_zipper_02}
      boxExtents=<15 20 20>
      constraintSetup {angularLimitsMax=<0 20 0>, angularLimitsMin=<0 -30 0>, angularTarget=<1 0 0>}
      localJointOffset=<1 0 0>
     }]
   }
   n5 = Control Rig!(
    Source=Component To Local(ComponentPose=@n4)
    Alpha
    isParachuteOpened:double=Get IsParachuteOpendFloat() -> (IsParachuteOpendFloat:double)
    istacticalSprinting:double=Get isTacticalSprintingFloat() -> (isTacticalSprintingFloat:double)
   ) -> (Pose) {controlRigAssetReference {blueprintRigClass=&/F_MED_EmberYearn_Post_CtrlRig.F_MED_EmberYearn_Post_CtrlRig_C}}
   Output Pose(Result=@n5) {layerGroup=None}
  ]}
 EventGraph {nodes [
   Event BlueprintUpdateAnimation()
   > n1 = Sequence!(execute) -> (then_0, then_1, then_2:Exec)
   > n2 = "Gravity Override State Function (Parameter Asset) (Nativized)"(
    CharacterPartAnimInstance=Self-Reference()
    GameTimeInSeconds=GetGameTimeInSeconds()
    Parameters=&/F_MED_EmberYearn_Parameters
   )
   > Set GravityOverride(GravityOverride=@n2)
   > Set RBAN Settings Out(RBAN Settings Out=@n2.OutRigidBodyAnimNodeParameters)
   Set IsParachuteOpendFloat(execute=@n1.then_1, IsParachuteOpendFloat:double="To Float (Boolean)"(InBool=Get bIsParachuteOpened())) -> (Output_Get:double)
   > Set isTacticalSprintingFloat(isTacticalSprintingFloat:double="To Float (Boolean)"(InBool=Get bIsTacticalSprinting())) -> (Output_Get:double)
  ]}
]
variables [
 GravityOverride {type=Vector}
 IsParachuteOpendFloat {type=number}
 isTacticalSprintingFloat {type=number}
 RBAN Settings Out {type=FortRigidBodyAnimNodeParameters}
]
