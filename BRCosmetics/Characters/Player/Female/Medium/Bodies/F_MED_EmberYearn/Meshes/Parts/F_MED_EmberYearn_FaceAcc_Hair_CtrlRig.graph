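# Control Rig for the EmberYearn female cosmetic's hair, posing hair bones from head and upper-body motion.
# Back-hair roots blend spine_05 and head rotation, a spherical pose reader on the head offsets left/right roots, and smoothed "is Targeting" rotates the right root.
# Weighted parent constraints then make most dynamic strand bones follow head, neck, spine, clavicles and arms, resembling skin weights.
# A debug function, disabled here, copies control transforms onto body bones.
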ControlRigBlueprint /BRCosmetics/Characters/Player/Female/Medium/Bodies/F_MED_EmberYearn/Meshes/Parts/F_MED_EmberYearn_FaceAcc_Hair_CtrlRig
graphs [
 New Function_ContainedGraph {nodes [
   n1 = RigVMFunctionEntryNode()
   RigVMFunction_ControlFlowBranch = &RigVMFunction_ControlFlowBranch(Condition=@n1.Debug, ExecuteContext=@n1.ExecuteContext)
   RigVMFunctionReturnNode(ExecuteContext=@RigVMFunction_ControlFlowBranch.Completed)
   DISPATCH_RigVMDispatch_ArrayIterator = RigVMDispatchNode(
    Array=/Script/ControlRig.RigUnit_CollectionGetAll(TypeToSearch=Control) {is=struct RigVMUnitNode}.Items
    ExecuteContext=@RigVMFunction_ControlFlowBranch.True
   ) {template="DISPATCH_RigVMDispatch_ArrayIterator(in Array,out Element,out Index,out Count,out Ratio)"}
   Replace = /Script/RigVM.RigVMFunction_NameReplace(Old=_ctrl, Name=@DISPATCH_RigVMDispatch_ArrayIterator.Element.Name) {is=struct RigVMUnitNode, template="Replace::Execute(in Name,in Old,in New,out Result)"}
   > &RigUnit_SetTransform(
    Item=(Type=Bone,Name="spine_01")
    Weight=1
    bInitial=False
    Item.Name=@Replace.Result
    Value=&RigUnit_GetTransform(Item=(Type=Control,Name="spine_01_ctrl"), bInitial=False, Item.Name=@DISPATCH_RigVMDispatch_ArrayIterator.Element.Name).Transform
   )
  ]}
 RigVMFunctionLibrary {nodes [ctrlRig = RigVMCollapseNode(Debug=False, ExecuteContext=())]}
 RigVMFunction_Sequence_1_ContainedGraph {nodes [
   n1 = RigVMFunctionEntryNode()
   RigVMFunction_Sequence = &RigVMFunction_Sequence(ExecuteContext=@n1.ExecuteContext)
   n2 = &RigVMFunction_Sequence(ExecuteContext=@RigVMFunction_Sequence.B)
   RigVMFunctionReturnNode(A=@RigVMFunction_Sequence.A, B=@n2.A, C=@n2.B)
  ]}
 RigVMModel {nodes [
   RigUnit_BeginExecution = &RigUnit_BeginExecution()
   ctrlRig = RigVMFunctionReferenceNode(Debug=false, ExecuteContext=@RigUnit_BeginExecution.ExecutePin)
   > RigVMFunction_Sequence = RigVMAggregateNode()
   Get Delta Ref_4 = RigVMFunctionReferenceNode(Item=(Type=Bone,Name="head"), ExecuteContext=@RigVMFunction_Sequence.A)
   > RigVMVariableNode() {variable=head rot z}
   &RigUnit_ParentConstraint(
    AdvancedSettings=(RotationOrderForFilter=XZY)
    Child=(Type=Bone,Name="base_HairA_BackL_01")
    Filter=(TranslationFilter=(bX=False,bY=False,bZ=False),RotationFilter=(bX=True,bY=True,bZ=True),ScaleFilter=(bX=False,bY=False,bZ=False))
    Parents=((Item=(Type=Bone,Name="spine_05"),Weight=0.8),(Item=(Type=Bone,Name="head"),Weight=0.2))
    Weight=1
    ExecutePin=@RigVMFunction_Sequence.B
   )
   > &RigUnit_ParentConstraint(
    AdvancedSettings=(RotationOrderForFilter=XZY)
    Child=(Type=Bone,Name="base_HairA_BackR_01")
    Filter=(TranslationFilter=(bX=False,bY=False,bZ=False),RotationFilter=(bX=True,bY=True,bZ=True),ScaleFilter=(bX=False,bY=False,bZ=False))
    Parents=((Item=(Type=Bone,Name="spine_05"),Weight=0.8),(Item=(Type=Bone,Name="head"),Weight=0.2))
    Weight=1
   )
   > &RigUnit_ParentConstraint(
    AdvancedSettings=(RotationOrderForFilter=XZY)
    Child=(Type=Bone,Name="base_HairA_Back_01")
    Filter=(TranslationFilter=(bX=False,bY=False,bZ=False),RotationFilter=(bX=True,bY=True,bZ=True),ScaleFilter=(bX=False,bY=False,bZ=False))
    Parents=((Item=(Type=Bone,Name="spine_05"),Weight=0.8),(Item=(Type=Bone,Name="head"),Weight=0.2))
    Weight=1
   )
   > n1 = &RigUnit_SphericalPoseReader(
    ActiveRegionScaleFactors=(PositiveWidth=2,NegativeWidth=3,PositiveHeight=3,NegativeHeight=5)
    ActiveRegionSize=0.1
    Debug=(bDrawDebug=false,bDraw2D=false,bDrawLocalAxes=false,DebugScale=15,DebugSegments=20,DebugThickness=0.25)
    DriverAxis=(X=1,Y=0,Z=0)
    DriverItem=(Type=Bone,Name="head")
    FalloffRegionScaleFactors=(PositiveWidth=0.8,NegativeWidth=1,PositiveHeight=1,NegativeHeight=1)
    FalloffSize=0.5
    FlipHeightScaling=false
    FlipWidthScaling=false
    OptionalParentItem=(Type=Bone,Name="neck_02")
    RotationOffset=(X=-90,Y=0,Z=10)
   )
   > &RigUnit_ModifyTransforms(
    ItemToModify=((Item=(Type=Bone,Name="base_HairA_BackR_01"),Transform=(Rotation=(X=0,Y=0,Z=0,W=1),Translation=(X=0,Y=-2,Z=5),Scale3D=(X=1,Y=1,Z=1))))
    WeightMinimum=0
    Weight=@n1.OutputParam
   )
   > &RigUnit_ModifyTransforms(
    ItemToModify=((Item=(Type=Bone,Name="base_HairA_BackL_01"),Transform=(Rotation=(X=0,Y=0,Z=0,W=1),Translation=(X=0,Y=-2,Z=3),Scale3D=(X=1,Y=1,Z=1))))
    WeightMinimum=0
    Weight=@n1.OutputParam
   )
   RigVMFunction_MathBoolToFloat = &RigVMFunction_MathBoolToFloat(Value=RigVMVariableNode() {variable=is Targeting}.Value)
   AlphaInterp = &RigVMFunction_AlphaInterp(InterpSpeedDecreasing=10, InterpSpeedIncreasing=10, Scale=1, bClampResult=False, Value=@RigVMFunction_MathBoolToFloat.Result)
   > &RigUnit_ModifyTransforms(
    ItemToModify=((Item=(Type=Bone,Name="base_HairA_BackR_01"),Transform=(Rotation=(X=0.0302,Y=-0.171,Z=0.171,W=0.9698),Translation=(X=0,Y=0,Z=0),Scale3D=(X=1,Y=1,Z=1))))
    WeightMinimum=0
    Weight=@AlphaInterp.Result
   )
   &RigUnit_ParentConstraint(
    AdvancedSettings=(RotationOrderForFilter=XZY)
    Child=(Type=Bone,Name="dyn_HairA_BackR_01")
    Parents=((Item=(Type=Bone,Name="head"),Weight=1.0))
    Weight=1
    ExecutePin=@RigVMFunction_Sequence.C
   )
   > &RigUnit_ParentConstraint(
    AdvancedSettings=(RotationOrderForFilter=XZY)
    Child=(Type=Bone,Name="dyn_HairA_BackR_02")
    Parents=((Item=(Type=Bone,Name="head"),Weight=0.2),(Item=(Type=Bone,Name="spine_05"),Weight=0.4),(Item=(Type=Bone,Name="clavicle_r"),Weight=0.4))
    Weight=1
   )
   > &RigUnit_ParentConstraint(
    AdvancedSettings=(RotationOrderForFilter=XZY)
    Child=(Type=Bone,Name="dyn_HairA_BackR_03")
    Parents=((Item=(Type=Bone,Name="neck_01"),Weight=0.1588),(Item=(Type=Bone,Name="head"),Weight=0.3395),(Item=(Type=Bone,Name="neck_02"),Weight=0.5018))
    Weight=1
   )
   > &RigUnit_ParentConstraint(
    AdvancedSettings=(RotationOrderForFilter=XZY)
    Child=(Type=Bone,Name="dyn_HairA_BackR_04")
    Parents=((Item=(Type=Bone,Name="upperarm_l"),Weight=0.0143),(Item=(Type=Bone,Name="clavicle_r"),Weight=0.4818),(Item=(Type=Bone,Name="spine_05"),Weight=0.5039))
    Weight=1
   )
   > &RigUnit_ParentConstraint(
    AdvancedSettings=(RotationOrderForFilter=XZY)
    Child=(Type=Bone,Name="dyn_HairA_BackR_05")
    Parents=((Item=(Type=Bone,Name="upperarm_r"),Weight=0.0456),(Item=(Type=Bone,Name="spine_05"),Weight=0.1299),(Item=(Type=Bone,Name="clavicle_r"),Weight=0.8245))
    Weight=1
   )
   > &RigUnit_ParentConstraint(
    AdvancedSettings=(RotationOrderForFilter=XZY)
    Child=(Type=Bone,Name="dyn_HairA_BackR_06")
    Parents=((Item=(Type=Bone,Name="upperarm_r"),Weight=0.045),(Item=(Type=Bone,Name="spine_04"),Weight=0.076),(Item=(Type=Bone,Name="spine_05"),Weight=0.257),(Item=(Type=Bone,Name="clavicle_r"),Weight=0.622))
    Weight=1
   )
   > &RigUnit_ParentConstraint(
    AdvancedSettings=(RotationOrderForFilter=XZY)
    Child=(Type=Bone,Name="dyn_HairA_Back_01")
    Parents=((Item=(Type=Bone,Name="head"),Weight=1.0))
    Weight=1
   )
   > &RigUnit_ParentConstraint(
    AdvancedSettings=(RotationOrderForFilter=XZY)
    Child=(Type=Bone,Name="dyn_HairA_Back_02")
    Parents=((Item=(Type=Bone,Name="head"),Weight=0.5),(Item=(Type=Bone,Name="spine_05"),Weight=0.5))
    Weight=1
   )
   > &RigUnit_ParentConstraint(
    AdvancedSettings=(RotationOrderForFilter=XZY)
    Child=(Type=Bone,Name="dyn_HairA_Back_03")
    Parents=((Item=(Type=Bone,Name="spine_05"),Weight=0.1265),(Item=(Type=Bone,Name="head"),Weight=0.1942),(Item=(Type=Bone,Name="neck_01"),Weight=0.2098),(Item=(Type=Bone,Name="neck_02"),Weight=0.4695))
    Weight=1
   )
   > &RigUnit_ParentConstraint(
    AdvancedSettings=(RotationOrderForFilter=XZY)
    Child=(Type=Bone,Name="dyn_HairA_Back_04")
    Parents=((Item=(Type=Bone,Name="clavicle_r"),Weight=0.0503),(Item=(Type=Bone,Name="spine_05"),Weight=0.9497))
    Weight=1
   )
   > &RigUnit_ParentConstraint(
    AdvancedSettings=(RotationOrderForFilter=XZY)
    Child=(Type=Bone,Name="dyn_HairA_Back_05")
    Parents=((Item=(Type=Bone,Name="clavicle_l"),Weight=0.0653),(Item=(Type=Bone,Name="clavicle_r"),Weight=0.3331),(Item=(Type=Bone,Name="spine_05"),Weight=0.6017))
    Weight=1
   )
   > &RigUnit_ParentConstraint(
    AdvancedSettings=(RotationOrderForFilter=XZY)
    Child=(Type=Bone,Name="dyn_HairA_Back_06")
    Parents=((Item=(Type=Bone,Name="clavicle_l"),Weight=0.057),(Item=(Type=Bone,Name="spine_04"),Weight=0.167),(Item=(Type=Bone,Name="clavicle_r"),Weight=0.214),(Item=(Type=Bone,Name="spine_05"),Weight=0.562))
    Weight=1
   )
   > &RigUnit_ParentConstraint(
    AdvancedSettings=(RotationOrderForFilter=XZY)
    Child=(Type=Bone,Name="dyn_HairA_Back_07")
    Parents="((Item=(Type=Bone,Name=\"spine_01\"),Weight=0.017),(Item=(Type=Bone,Name=\"clavicle_l\"),Weight=0.025),(Item=(Type=Bone,Name=\"spine_02\"),Weight=0.046),(Item=(Type=Bone,Name=\"spine_03\"),Weight=0.072),(Item=(Type=Bone,Name=\"clavicle_r\"),Weight=0.101),(Item=(Type ...(truncated)"
    Weight=1
   )
   > &RigUnit_ParentConstraint(
    AdvancedSettings=(RotationOrderForFilter=XZY)
    Child=(Type=Bone,Name="dyn_HairA_BackL_01")
    Parents=((Item=(Type=Bone,Name="head"),Weight=1.0))
    Weight=1
   )
   > &RigUnit_ParentConstraint(
    AdvancedSettings=(RotationOrderForFilter=XZY)
    Child=(Type=Bone,Name="dyn_HairA_BackL_02")
    Parents=((Item=(Type=Bone,Name="head"),Weight=0.2),(Item=(Type=Bone,Name="spine_05"),Weight=0.4),(Item=(Type=Bone,Name="clavicle_l"),Weight=0.4))
    Weight=1
   )
   > &RigUnit_ParentConstraint(
    AdvancedSettings=(RotationOrderForFilter=XZY)
    Child=(Type=Bone,Name="dyn_HairA_BackL_03")
    Parents=((Item=(Type=Bone,Name="head"),Weight=0.1299),(Item=(Type=Bone,Name="neck_01"),Weight=0.1706),(Item=(Type=Bone,Name="neck_02"),Weight=0.3251),(Item=(Type=Bone,Name="spine_05"),Weight=0.3745))
    Weight=1
   )
   > &RigUnit_ParentConstraint(
    AdvancedSettings=(RotationOrderForFilter=XZY)
    Child=(Type=Bone,Name="dyn_HairA_BackL_04")
    Parents=((Item=(Type=Bone,Name="clavicle_l"),Weight=0.434),(Item=(Type=Bone,Name="spine_05"),Weight=0.566))
    Weight=1
   )
   > &RigUnit_ParentConstraint(
    AdvancedSettings=(RotationOrderForFilter=XZY)
    Child=(Type=Bone,Name="dyn_HairA_BackL_05")
    Parents=((Item=(Type=Bone,Name="upperarm_l"),Weight=0.025),(Item=(Type=Bone,Name="spine_04"),Weight=0.025),(Item=(Type=Bone,Name="spine_05"),Weight=0.235),(Item=(Type=Bone,Name="clavicle_l"),Weight=0.715))
    Weight=1
   )
   > &RigUnit_ParentConstraint(
    AdvancedSettings=(RotationOrderForFilter=XZY)
    Child=(Type=Bone,Name="dyn_HairA_BackL_06")
    Parents=((Item=(Type=Bone,Name="spine_03"),Weight=0.0101),(Item=(Type=Bone,Name="upperarm_l"),Weight=0.0202),(Item=(Type=Bone,Name="spine_04"),Weight=0.1948),(Item=(Type=Bone,Name="spine_05"),Weight=0.329),(Item=(Type=Bone,Name="clavicle_l"),Weight=0.446))
    Weight=1
   )
   > &RigUnit_ParentConstraint(
    AdvancedSettings=(RotationOrderForFilter=XZY)
    Child=(Type=Bone,Name="dyn_HairA_mid_L_01")
    Parents=((Item=(Type=Bone,Name="clavicle_l"),Weight=0.0146),(Item=(Type=Bone,Name="spine_05"),Weight=0.0222),(Item=(Type=Bone,Name="head"),Weight=0.9632))
    Weight=1
   )
   > &RigUnit_ParentConstraint(
    AdvancedSettings=(RotationOrderForFilter=XZY)
    Child=(Type=Bone,Name="dyn_HairA_mid_L_02")
    Parents=((Item=(Type=Bone,Name="spine_05"),Weight=0.1255),(Item=(Type=Bone,Name="clavicle_l"),Weight=0.1505),(Item=(Type=Bone,Name="head"),Weight=0.724))
    Weight=1
   )
   > &RigUnit_ParentConstraint(
    AdvancedSettings=(RotationOrderForFilter=XZY)
    Child=(Type=Bone,Name="dyn_HairA_mid_L_03")
    Parents=((Item=(Type=Bone,Name="clavicle_l"),Weight=0.2205),(Item=(Type=Bone,Name="spine_05"),Weight=0.3162),(Item=(Type=Bone,Name="head"),Weight=0.4634))
    Weight=1
   )
   > &RigUnit_ParentConstraint(
    AdvancedSettings=(RotationOrderForFilter=XZY)
    Child=(Type=Bone,Name="dyn_HairA_mid_L_04")
    Parents=((Item=(Type=Bone,Name="head"),Weight=0.2571),(Item=(Type=Bone,Name="clavicle_l"),Weight=0.3058),(Item=(Type=Bone,Name="spine_05"),Weight=0.4371))
    Weight=1
   )
   > &RigUnit_ParentConstraint(
    AdvancedSettings=(RotationOrderForFilter=XZY)
    Child=(Type=Bone,Name="dyn_HairA_mid_L_05")
    Parents=((Item=(Type=Bone,Name="head"),Weight=0.0444),(Item=(Type=Bone,Name="clavicle_l"),Weight=0.2496),(Item=(Type=Bone,Name="spine_05"),Weight=0.706))
    Weight=1
   )
   > &RigUnit_ParentConstraint(
    AdvancedSettings=(RotationOrderForFilter=XZY)
    Child=(Type=Bone,Name="dyn_HairA_mid_L_06")
    Parents=((Item=(Type=Bone,Name="clavicle_l"),Weight=0.0608),(Item=(Type=Bone,Name="spine_05"),Weight=0.9392))
    Weight=1
   )
   > &RigUnit_ParentConstraint(
    AdvancedSettings=(RotationOrderForFilter=XZY)
    Child=(Type=Bone,Name="dyn_HairA_R_01")
    Parents=((Item=(Type=Bone,Name="clavicle_r"),Weight=0.0071),(Item=(Type=Bone,Name="spine_05"),Weight=0.0171),(Item=(Type=Bone,Name="head"),Weight=0.9758))
    Weight=1
   )
   > &RigUnit_ParentConstraint(
    AdvancedSettings=(RotationOrderForFilter=XZY)
    Child=(Type=Bone,Name="dyn_HairA_R_02")
    Parents=((Item=(Type=Bone,Name="clavicle_r"),Weight=0.0302),(Item=(Type=Bone,Name="spine_05"),Weight=0.1254),(Item=(Type=Bone,Name="head"),Weight=0.8444))
    Weight=1
   )
   > &RigUnit_ParentConstraint(
    AdvancedSettings=(RotationOrderForFilter=XZY)
    Child=(Type=Bone,Name="dyn_HairA_R_03")
    Parents=((Item=(Type=Bone,Name="clavicle_r"),Weight=0.0806),(Item=(Type=Bone,Name="spine_05"),Weight=0.3731),(Item=(Type=Bone,Name="head"),Weight=0.5463))
    Weight=1
   )
   > &RigUnit_ParentConstraint(
    AdvancedSettings=(RotationOrderForFilter=XZY)
    Child=(Type=Bone,Name="dyn_HairA_R_04")
    Parents=((Item=(Type=Bone,Name="clavicle_r"),Weight=0.1935),(Item=(Type=Bone,Name="head"),Weight=0.2494),(Item=(Type=Bone,Name="spine_05"),Weight=0.5571))
    Weight=1
   )
   > &RigUnit_ParentConstraint(
    AdvancedSettings=(RotationOrderForFilter=XZY)
    Child=(Type=Bone,Name="dyn_HairA_R_05")
    Parents=((Item=(Type=Bone,Name="head"),Weight=0.0715),(Item=(Type=Bone,Name="clavicle_r"),Weight=0.1893),(Item=(Type=Bone,Name="spine_05"),Weight=0.7393))
    Weight=1
   )
   > &RigUnit_ParentConstraint(
    AdvancedSettings=(RotationOrderForFilter=XZY)
    Child=(Type=Bone,Name="dyn_HairA_R_06")
    Parents=((Item=(Type=Bone,Name="head"),Weight=0.0037),(Item=(Type=Bone,Name="clavicle_r"),Weight=0.0659),(Item=(Type=Bone,Name="spine_05"),Weight=0.9304))
    Weight=1
   )
   > &RigUnit_ParentConstraint(
    AdvancedSettings=(RotationOrderForFilter=XZY)
    Child=(Type=Bone,Name="dyn_HairA_L_01")
    Parents=((Item=(Type=Bone,Name="clavicle_l"),Weight=0.0242),(Item=(Type=Bone,Name="spine_05"),Weight=0.0374),(Item=(Type=Bone,Name="head"),Weight=0.9384))
    Weight=1
   )
   > &RigUnit_ParentConstraint(
    AdvancedSettings=(RotationOrderForFilter=XZY)
    Child=(Type=Bone,Name="dyn_HairA_L_02")
    Parents=((Item=(Type=Bone,Name="clavicle_l"),Weight=0.1153),(Item=(Type=Bone,Name="spine_05"),Weight=0.1209),(Item=(Type=Bone,Name="head"),Weight=0.7638))
    Weight=1
   )
   > &RigUnit_ParentConstraint(
    AdvancedSettings=(RotationOrderForFilter=XZY)
    Child=(Type=Bone,Name="dyn_HairA_L_03")
    Parents=((Item=(Type=Bone,Name="spine_05"),Weight=0.2309),(Item=(Type=Bone,Name="clavicle_l"),Weight=0.2819),(Item=(Type=Bone,Name="head"),Weight=0.4872))
    Weight=1
   )
   > &RigUnit_ParentConstraint(
    AdvancedSettings=(RotationOrderForFilter=XZY)
    Child=(Type=Bone,Name="dyn_HairA_L_04")
    Parents=((Item=(Type=Bone,Name="head"),Weight=0.2852),(Item=(Type=Bone,Name="spine_05"),Weight=0.3366),(Item=(Type=Bone,Name="clavicle_l"),Weight=0.3782))
    Weight=1
   )
   > &RigUnit_ParentConstraint(
    AdvancedSettings=(RotationOrderForFilter=XZY)
    Child=(Type=Bone,Name="dyn_HairA_L_05")
    Parents=((Item=(Type=Bone,Name="head"),Weight=0.0821),(Item=(Type=Bone,Name="spine_05"),Weight=0.4099),(Item=(Type=Bone,Name="clavicle_l"),Weight=0.508))
    Weight=1
   )
   > &RigUnit_ParentConstraint(
    AdvancedSettings=(RotationOrderForFilter=XZY)
    Child=(Type=Bone,Name="dyn_HairA_L_06")
    Parents=((Item=(Type=Bone,Name="head"),Weight=0.0036),(Item=(Type=Bone,Name="spine_04"),Weight=0.0125),(Item=(Type=Bone,Name="spine_05"),Weight=0.3641),(Item=(Type=Bone,Name="clavicle_l"),Weight=0.6198))
    Weight=1
   )
   > &RigUnit_ParentConstraint(
    AdvancedSettings=(RotationOrderForFilter=XZY)
    Child=(Type=Bone,Name="dyn_HairA_B_R_01")
    Parents=((Item=(Type=Bone,Name="clavicle_r"),Weight=0.2708),(Item=(Type=Bone,Name="spine_05"),Weight=0.2783),(Item=(Type=Bone,Name="head"),Weight=0.4509))
    Weight=1
   )
   > &RigUnit_ParentConstraint(
    AdvancedSettings=(RotationOrderForFilter=XZY)
    Child=(Type=Bone,Name="dyn_HairA_B_R_02")
    Parents=((Item=(Type=Bone,Name="spine_04"),Weight=0.0047),(Item=(Type=Bone,Name="head"),Weight=0.1445),(Item=(Type=Bone,Name="clavicle_r"),Weight=0.262),(Item=(Type=Bone,Name="spine_05"),Weight=0.5889))
    Weight=1
   )
   > &RigUnit_ParentConstraint(
    AdvancedSettings=(RotationOrderForFilter=XZY)
    Child=(Type=Bone,Name="dyn_HairA_B_R_03")
    Parents=((Item=(Type=Bone,Name="spine_04"),Weight=0.0155),(Item=(Type=Bone,Name="head"),Weight=0.0755),(Item=(Type=Bone,Name="clavicle_r"),Weight=0.313),(Item=(Type=Bone,Name="spine_05"),Weight=0.596))
    Weight=1
   )
   > &RigUnit_ParentConstraint(
    AdvancedSettings=(RotationOrderForFilter=XZY)
    Child=(Type=Bone,Name="dyn_HairA_B_R_04")
    Parents=((Item=(Type=Bone,Name="spine_04"),Weight=0.0184),(Item=(Type=Bone,Name="head"),Weight=0.0233),(Item=(Type=Bone,Name="clavicle_r"),Weight=0.3676),(Item=(Type=Bone,Name="spine_05"),Weight=0.5907))
    Weight=1
   )
   > &RigUnit_ParentConstraint(
    AdvancedSettings=(RotationOrderForFilter=XZY)
    Child=(Type=Bone,Name="dyn_HairA_A_R_01")
    Parents=((Item=(Type=Bone,Name="clavicle_r"),Weight=0.0513),(Item=(Type=Bone,Name="spine_05"),Weight=0.3053),(Item=(Type=Bone,Name="head"),Weight=0.6433))
    Weight=1
   )
   > &RigUnit_ParentConstraint(
    AdvancedSettings=(RotationOrderForFilter=XZY)
    Child=(Type=Bone,Name="dyn_HairA_A_R_02")
    Parents=((Item=(Type=Bone,Name="clavicle_r"),Weight=0.0453),(Item=(Type=Bone,Name="spine_05"),Weight=0.352),(Item=(Type=Bone,Name="head"),Weight=0.6027))
    Weight=1
   )
   > &RigUnit_ParentConstraint(
    AdvancedSettings=(RotationOrderForFilter=XZY)
    Child=(Type=Bone,Name="dyn_HairA_A_R_03")
    Parents=((Item=(Type=Bone,Name="clavicle_r"),Weight=0.0162),(Item=(Type=Bone,Name="spine_05"),Weight=0.4208),(Item=(Type=Bone,Name="head"),Weight=0.5629))
    Weight=1
   )
   > &RigUnit_ParentConstraint(
    AdvancedSettings=(RotationOrderForFilter=XZY)
    Child=(Type=Bone,Name="dyn_HairA_A_R_04")
    Parents=((Item=(Type=Bone,Name="head"),Weight=0.4774),(Item=(Type=Bone,Name="spine_05"),Weight=0.5226))
    Weight=1
   )
   > &RigUnit_ParentConstraint(
    AdvancedSettings=(RotationOrderForFilter=XZY)
    Child=(Type=Bone,Name="dyn_HairA_front_uppr_A_R_01")
    Parents=((Item=(Type=Bone,Name="clavicle_r"),Weight=0.0441),(Item=(Type=Bone,Name="spine_05"),Weight=0.1748),(Item=(Type=Bone,Name="head"),Weight=0.7811))
    Weight=1
   )
   > &RigUnit_ParentConstraint(
    AdvancedSettings=(RotationOrderForFilter=XZY)
    Child=(Type=Bone,Name="dyn_HairA_front_uppr_A_R_03")
    Parents=((Item=(Type=Bone,Name="clavicle_l"),Weight=0.0021),(Item=(Type=Bone,Name="spine_04"),Weight=0.0035),(Item=(Type=Bone,Name="clavicle_r"),Weight=0.1329),(Item=(Type=Bone,Name="spine_05"),Weight=0.3853),(Item=(Type=Bone,Name="head"),Weight=0.4762))
    Weight=1
   )
   > &RigUnit_ParentConstraint(
    AdvancedSettings=(RotationOrderForFilter=XZY)
    Child=(Type=Bone,Name="dyn_HairA_front_uppr_A_R_02")
    Parents=((Item=(Type=Bone,Name="clavicle_r"),Weight=0.0922),(Item=(Type=Bone,Name="spine_05"),Weight=0.3151),(Item=(Type=Bone,Name="head"),Weight=0.5927))
    Weight=1
   )
   &RigUnit_ModifyTransforms(
    ExecutePin=()
    ItemToModify="((Item=(Type=Bone,Name=\"dyn_HairA_L_03\"),Transform=(Rotation=(X=0,Y=0,Z=0,W=1),Translation=(X=0,Y=1,Z=0),Scale3D=(X=1,Y=1,Z=1))),(Item=(Type=Bone,Name=\"dyn_HairA_mid_L_04\"),Transform=(Rotation=(X=0,Y=0,Z=0,W=1),Translation=(X=0,Y=1,Z=0),Scale3D=(X=1,Y=1,Z= ...(truncated)"
    Weight=1
    WeightMinimum=0
   )
   &RigVMFunction_MathDoubleRemap(SourceMaximum=-50, SourceMinimum=0, TargetMaximum=1, TargetMinimum=0, Value=@"Get Delta Ref_4".Result.Rotation.Yaw)
  ]}
]
hierarchy [
 root {loc=<0 0 0>, rot=<0 0 0 1>}
 pelvis {parent=root, loc=<0 -0.5301 98.0571>, rot=<0.0127 0.707 -0.0127 -0.707>}
 spine_01 {parent=pelvis, loc=<6.1208 -0.9328 0>, rot=<0 0 0.034 -0.9994>}
 spine_02 {parent=spine_01, loc=<6.5997 -0.2095 0>, rot=<0 0 0.0238 -0.9997>}
 spine_03 {parent=spine_02, loc=<6.4914 -0.1237 0>, rot=<0 0 0.0149 -0.9999>}
 spine_04 {parent=spine_03, loc=<6.9598 0.0794 0>, rot=<0 0 -0.0132 -0.9999>}
 spine_05 {parent=spine_04, loc=<6.9987 0 0>, rot=<0 0 -0.0015 -1>}
 neck_01 {parent=spine_05, loc=<11.3138 0.1529 0>, rot=<0 0 -0.2003 -0.9797>}
 neck_02 {parent=neck_01, loc=<4.5627 0 0>, rot=<0 0 0.0504 -0.9987>}
 head {parent=neck_02, loc=<5.3914 0 0>, rot=<0 0 0.115 -0.9934>}
 base_HairA_BackR_01 {parent=head, loc=<6.1965 -3.2869 6.03>, rot=<0.0541 -0.4196 0.9044 -0.0558>}
 dyn_HairA_BackR_01 {parent=base_HairA_BackR_01, loc=<0 0 0>, rot=<0.0929 -0.0335 0.0702 -0.9926>}
 dyn_HairA_BackR_02 {parent=dyn_HairA_BackR_01, loc=<5.6131 0 0>, rot=<-0.0087 0.0592 -0.0531 -0.9968>}
 dyn_HairA_BackR_03 {parent=dyn_HairA_BackR_02, loc=<5.6131 0 0>, rot=<-0.0175 0.1591 -0.0919 -0.9828>}
 dyn_HairA_BackR_04 {parent=dyn_HairA_BackR_03, loc=<5.6131 0 0>, rot=<-0.0035 0.0307 -0.0499 -0.9983>}
 dyn_HairA_BackR_05 {parent=dyn_HairA_BackR_04, loc=<5.6131 0 0>, rot=<-0.0101 -0.2749 0.0744 -0.9585>}
 dyn_HairA_BackR_06 {parent=dyn_HairA_BackR_05, loc=<5.6131 0 0>, rot=<0 0 0 1>}
 base_HairA_Back_01 {parent=head, loc=<6.5329 -6.4927 0.8092>, rot=<0.0703 -0.7036 0.7036 -0.0703>}
 dyn_HairA_Back_01 {parent=base_HairA_Back_01, loc=<0 0 0>, rot=<0 0 0 1>}
 dyn_HairA_Back_02 {parent=dyn_HairA_Back_01, loc=<5.8294 0 0>, rot=<0 0.01 0.002 -0.9999>}
 dyn_HairA_Back_03 {parent=dyn_HairA_Back_02, loc=<5.8294 0 0>, rot=<-0.0009 0.0664 0.0091 -0.9978>}
 dyn_HairA_Back_04 {parent=dyn_HairA_Back_03, loc=<5.8294 0 0>, rot=<0.0018 -0.0682 0.0044 -0.9977>}
 dyn_HairA_Back_05 {parent=dyn_HairA_Back_04, loc=<5.8294 0 0>, rot=<0.0038 -0.0964 0.0081 -0.9953>}
 dyn_HairA_Back_06 {parent=dyn_HairA_Back_05, loc=<5.8294 0 0>, rot=<0.0045 -0.0836 0.0064 -0.9965>}
 dyn_HairA_Back_07 {parent=dyn_HairA_Back_06, loc=<5.8294 0 0>, rot=<0 0 0 1>}
 base_HairA_BackL_01 {parent=head, loc=<5.8103 -4.5167 -5.7556>, rot=<0.0536 -0.9377 0.3388 -0.0554>}
 dyn_HairA_BackL_01 {parent=base_HairA_BackL_01, loc=<0 0 0>, rot=<-0.175 -0.0147 -0.0138 -0.9844>}
 dyn_HairA_BackL_02 {parent=dyn_HairA_BackL_01, loc=<5.7229 0 0>, rot=<0.0053 0.0611 -0.007 -0.9981>}
 dyn_HairA_BackL_03 {parent=dyn_HairA_BackL_02, loc=<5.7229 0 0>, rot=<0.017 0.1025 0.0601 -0.9928>}
 dyn_HairA_BackL_04 {parent=dyn_HairA_BackL_03, loc=<5.7229 0 0>, rot=<0.0007 -0.0321 0.1166 -0.9927>}
 dyn_HairA_BackL_05 {parent=dyn_HairA_BackL_04, loc=<5.7229 0 0>, rot=<0.0326 -0.2076 -0.044 -0.9767>}
 dyn_HairA_BackL_06 {parent=dyn_HairA_BackL_05, loc=<5.723 0 0>, rot=<0 0 0 1>}
 base_HairA_front_L_01 {parent=head, loc=<1.5231 6.9804 -6.776>, rot=<0.1002 -0.9739 -0.1582 0.1282>}
 dyn_HairA_front_L_01 {parent=base_HairA_front_L_01, loc=<0 0 0>, rot=<0 0 0 1>}
 dyn_HairA_front_L_02 {parent=dyn_HairA_front_L_01, loc=<2.9883 0 0>, rot=<0.006 0.0954 -0.0123 -0.9953>}
 dyn_HairA_front_L_03 {parent=dyn_HairA_front_L_02, loc=<2.9796 0 0>, rot=<-0.01 0.1657 -0.2051 -0.9646>}
 dyn_HairA_front_L_04 {parent=dyn_HairA_front_L_03, loc=<3.0054 0 0>, rot=<0 0 0 1>}
 base_HairA_L_01 {parent=head, loc=<2.958 -0.0365 -7.9103>, rot=<0.0795 -0.9955 -0.004 -0.0505>}
 dyn_HairA_L_01 {parent=base_HairA_L_01, loc=<0 0 0>, rot=<0 0 0 1>}
 dyn_HairA_L_02 {parent=dyn_HairA_L_01, loc=<3.5871 0 0>, rot=<-0.0047 0.031 -0.0352 -0.9989>}
 dyn_HairA_L_03 {parent=dyn_HairA_L_02, loc=<3.5669 0 0>, rot=<-0.0095 -0.0209 -0.066 -0.9976>}
 dyn_HairA_L_04 {parent=dyn_HairA_L_03, loc=<3.6702 0 0>, rot=<-0.0217 0.0029 -0.1734 -0.9846>}
 dyn_HairA_L_05 {parent=dyn_HairA_L_04, loc=<3.4914 0 0>, rot=<-0.04 0.0405 -0.2344 -0.9705>}
 dyn_HairA_L_06 {parent=dyn_HairA_L_05, loc=<3.6419 0 0>, rot=<0 0 0 1>}
 base_HairA_R_01 {parent=head, loc=<4.0974 1.6889 7.7932>, rot=<-0.0597 0.1736 0.9829 0.0156>}
 dyn_HairA_R_01 {parent=base_HairA_R_01, loc=<0 0 0>, rot=<0 0 0 1>}
 dyn_HairA_R_02 {parent=dyn_HairA_R_01, loc=<4.1888 0 0>, rot=<0.0046 -0.0065 -0.0397 -0.9992>}
 dyn_HairA_R_03 {parent=dyn_HairA_R_02, loc=<4.1888 0 0>, rot=<-0.009 -0.0004 0.0716 -0.9974>}
 dyn_HairA_R_04 {parent=dyn_HairA_R_03, loc=<4.1888 0 0>, rot=<-0.0224 0.0246 0.1239 -0.9917>}
 dyn_HairA_R_05 {parent=dyn_HairA_R_04, loc=<4.1888 0 0>, rot=<-0.0145 0.2124 0.1913 -0.9582>}
 dyn_HairA_R_06 {parent=dyn_HairA_R_05, loc=<4.1888 0 0>, rot=<0 0 0 1>}
 dyn_HairA_B_R_01 {parent=dyn_HairA_R_03, loc=<1.187 2.5609 0.6836>, rot=<-0.0875 0.0817 -0.0276 -0.9924>}
 dyn_HairA_B_R_02 {parent=dyn_HairA_B_R_01, loc=<3.053 0 0>, rot=<-0.0102 0.0339 0.1942 -0.9803>}
 dyn_HairA_B_R_03 {parent=dyn_HairA_B_R_02, loc=<3.0479 0 0>, rot=<-0.0298 0.0403 0.1709 -0.984>}
 dyn_HairA_B_R_04 {parent=dyn_HairA_B_R_03, loc=<3.0539 0 0>, rot=<0 0 0 1>}
 dyn_HairA_A_R_01 {parent=dyn_HairA_R_03, loc=<0.4159 -1.8399 1.2711>, rot=<0.1438 0.173 0.1054 -0.9686>}
 dyn_HairA_A_R_02 {parent=dyn_HairA_A_R_01, loc=<2.7423 0 0>, rot=<0.0077 -0.054 0.1315 -0.9898>}
 dyn_HairA_A_R_03 {parent=dyn_HairA_A_R_02, loc=<2.8637 0 0>, rot=<0.097 -0.3671 -0.1681 -0.9097>}
 dyn_HairA_A_R_04 {parent=dyn_HairA_A_R_03, loc=<2.6204 0 0>, rot=<0 0 0 1>}
 dyn_HairA_front_uppr_A_R_01 {parent=dyn_HairA_R_02, loc=<1.7165 -0.4057 1.7847>, rot=<0.1518 0.1726 -0.5859 -0.7771>}
 dyn_HairA_front_uppr_A_R_02 {parent=dyn_HairA_front_uppr_A_R_01, loc=<3.0678 0 0>, rot=<-0.0535 0 -0.129 -0.9902>}
 dyn_HairA_front_uppr_A_R_03 {parent=dyn_HairA_front_uppr_A_R_02, loc=<3.0433 0 0>, rot=<0 0 0 1>}
 base_HairA_mid_L_01 {parent=head, loc=<2.3297 3.1813 -7.2358>, rot=<0.1246 -0.9869 -0.0928 -0.0428>}
 dyn_HairA_mid_L_01 {parent=base_HairA_mid_L_01, loc=<0 0 0>, rot=<0 0 0 1>}
 dyn_HairA_mid_L_02 {parent=dyn_HairA_mid_L_01, loc=<4.0333 0 0>, rot=<-0.0002 -0.0091 -0.0191 -0.9998>}
 dyn_HairA_mid_L_03 {parent=dyn_HairA_mid_L_02, loc=<4.0333 0 0>, rot=<0.0001 0.0009 -0.1175 -0.9931>}
 dyn_HairA_mid_L_04 {parent=dyn_HairA_mid_L_03, loc=<4.0332 0 0>, rot=<-0.0372 -0.1771 -0.2022 -0.9625>}
 dyn_HairA_mid_L_05 {parent=dyn_HairA_mid_L_04, loc=<4.0334 0 0>, rot=<-0.0339 -0.1873 -0.1746 -0.9661>}
 dyn_HairA_mid_L_06 {parent=dyn_HairA_mid_L_05, loc=<4.0333 0 0>, rot=<0 0 0 1>}
 head_hair_back_collision {parent=head, loc=<0 0 0>, rot=<0 0 0 1>}
 clavicle_l {parent=spine_05, loc=<5.4831 -0.1053 -3.2946>, rot=<0.0465 -0.7289 -0.009 -0.6829>}
 clavicle_l__hair_back_collision {parent=clavicle_l, loc=<0 0 0>, rot=<0 0 0 1>}
 upperarm_l {parent=clavicle_l, loc=<9.8658 -1.2136 -0.4105>, rot=<0.0164 -0.3245 -0.0269 -0.9454>}
 clavicle_r {parent=spine_05, loc=<5.4831 -0.1053 3.2946>, rot=<0.7289 0.0465 -0.6829 0.009>}
 clavicle_r__hair_back_collision {parent=clavicle_r, loc=<0 0 0>, rot=<0 0 0 1>}
 upperarm_r {parent=clavicle_r, loc=<-9.8658 1.2136 0.4105>, rot=<0.0164 -0.3245 -0.0269 -0.9454>}
 spine_04__hair_back_collision {parent=spine_04, loc=<0 0 0>, rot=<0 0 0 1>}
 spine_03_ctrl {controlType=EULER_TRANSFORM, type=Control, loc=<0 0 0>, rot=<0 0 0 1>}
 spine_04_ctrl {controlType=EULER_TRANSFORM, parent=spine_03_ctrl, type=Control, loc=<0 0 0>, rot=<0 0 0 1>}
 spine_05_ctrl {controlType=EULER_TRANSFORM, parent=spine_04_ctrl, type=Control, loc=<0 0 0>, rot=<0 0 0 1>}
 neck_01_ctrl {controlType=EULER_TRANSFORM, parent=spine_05_ctrl, type=Control, loc=<0 0 0>, rot=<0 0 0 1>}
 neck_02_ctrl {controlType=EULER_TRANSFORM, parent=neck_01_ctrl, type=Control, loc=<0 0 0>, rot=<0 0 0 1>}
 head_ctrl {controlType=EULER_TRANSFORM, parent=neck_02_ctrl, type=Control, loc=<0 0 0>, rot=<0 0 0 1>}
 clavicle_l_ctrl {controlType=EULER_TRANSFORM, parent=spine_05_ctrl, type=Control, loc=<0 0 0>, rot=<0 0 0 1>}
 clavicle_r_ctrl {controlType=EULER_TRANSFORM, parent=spine_05_ctrl, type=Control, loc=<0 0 0>, rot=<0 0 0 1>}
]
variables [
 BackBiling {type=bool, default=False}
 head rot z {type=double, default=0}
 is Targeting {type=bool, default=True}
]
