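# Post-simulation Control Rig for the EmberYearn female character's hair, applying corrective transforms to the dynamic hair bones under the head.
# Head yaw and roll deltas are remapped into weights, combined with back bling, targeting and skydiving flags to drive back-hair adjustments.
# Alpha interpolation smooths the flag-driven weights, while other fixed corrections always apply at full weight.
# A debug function that copies control transforms onto matching bones is defined but unused in the main graph.
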ControlRigBlueprint /BRCosmetics/Characters/Player/Female/Medium/Bodies/F_MED_EmberYearn/Meshes/Parts/F_MED_EmberYearn_FaceAcc_Hair_PostSimCtrlRig
graphs [
 Multiply_1_ContainedGraph {nodes [
   n1 = RigVMFunctionEntryNode()
   Multiply = &RigVMFunction_MathDoubleMul(A=@n1.A, B=@n1.B)
   RigVMFunctionReturnNode(Result=&RigVMFunction_MathDoubleMul(B=@n1.C, A=@Multiply.Result).Result)
  ]}
 New Function_ContainedGraph {nodes [
   n1 = RigVMFunctionEntryNode()
   RigVMFunction_ControlFlowBranch = &RigVMFunction_ControlFlowBranch(Condition=@n1.Debug, ExecuteContext=@n1.ExecuteContext)
   RigVMFunctionReturnNode(ExecuteContext=@RigVMFunction_ControlFlowBranch.Completed)
   DISPATCH_RigVMDispatch_ArrayIterator = RigVMDispatchNode(
    Array=/Script/ControlRig.RigUnit_CollectionGetAll(TypeToSearch=Control) {is=struct RigVMUnitNode}.Items
    ExecuteContext=@RigVMFunction_ControlFlowBranch.True
   ) {template="DISPATCH_RigVMDispatch_ArrayIterator(in Array,out Element,out Index,out Count,out Ratio)"}
   Replace = /Script/RigVM.RigVMFunction_NameReplace(Old=_ctrl, Name=@DISPATCH_RigVMDispatch_ArrayIterator.Element.Name) {is=struct RigVMUnitNode, template="Replace::Execute(in Name,in Old,in New,out Result)"}
   > &RigUnit_SetTransform(
    Item=(Type=Bone,Name="spine_01")
    Weight=1
    bInitial=False
    Item.Name=@Replace.Result
    Value=&RigUnit_GetTransform(Item=(Type=Control,Name="spine_01_ctrl"), bInitial=False, Item.Name=@DISPATCH_RigVMDispatch_ArrayIterator.Element.Name).Transform
   )
  ]}
 RigVMFunctionLibrary {nodes [ctrlRig = RigVMCollapseNode(Debug=False, ExecuteContext=())]}
 RigVMModel {nodes [
   RigUnit_BeginExecution = &RigUnit_BeginExecution()
   &RigUnit_ModifyTransforms(
    ItemToModify=((Item=(Type=Bone,Name="head"),Transform=(Rotation=(X=0,Y=0,Z=0,W=1),Translation=(X=0,Y=0,Z=0),Scale3D=(X=1,Y=1,Z=1))))
    Weight=1
    WeightMinimum=0
    ExecutePin=@RigUnit_BeginExecution.ExecutePin
   )
   > Get Delta Ref_4 = RigVMFunctionReferenceNode(Item=(Type=Bone,Name="head"))
   > RigVMVariableNode(Value=&RigVMFunction_MathDoubleRemap(SourceMaximum=-50, SourceMinimum=0, TargetMaximum=1, TargetMinimum=0, Value=@"Get Delta Ref_4".Result.Rotation.Yaw).Result) {variable=head rot z}
   Multiply = RigVMAggregateNode(
    B=&RigVMFunction_MathBoolToFloat(Value=RigVMVariableNode() {variable=BackBiling}.Value).Result
    A=&RigVMFunction_MathBoolToFloat(Value=RigVMVariableNode() {variable=is Targeting}.Value).Result
    C=RigVMVariableNode() {variable=head rot z}.Value
   )
   n1 = &RigVMFunction_AlphaInterp(InterpSpeedDecreasing=10, InterpSpeedIncreasing=10, Scale=1, bClampResult=False, Value=@Multiply.Result)
   > &RigUnit_ModifyTransforms(
    ItemToModify="((Item=(Type=Bone,Name=\"dyn_HairA_BackR_03\"),Transform=(Rotation=(X=-0.4226,Y=0,Z=0,W=0.9063),Translation=(X=0,Y=0,Z=0),Scale3D=(X=1,Y=1,Z=1))),(Item=(Type=Bone,Name=\"dyn_HairA_BackR_02\"),Transform=(Rotation=(X=0,Y=0,Z=0,W=1),Translation=(X=-1,Y=0,Z=0),Sca ...(truncated)"
    WeightMinimum=0
    Weight=@n1.Result
   )
   n2 = RigVMVariableNode() {variable=BackBiling}
   n3 = RigVMVariableNode() {variable=Skydivining}
   n4 = &RigVMFunction_AlphaInterp(
    InterpSpeedDecreasing=10
    InterpSpeedIncreasing=10
    Scale=1
    bClampResult=False
    Value=&RigVMFunction_MathDoubleMul(A=&RigVMFunction_MathBoolToFloat(Value=@n2.Value).Result, B=&RigVMFunction_MathBoolToFloat(Value=@n3.Value).Result).Result
   )
   > &RigUnit_ModifyTransforms(
    ItemToModify="((Item=(Type=Bone,Name=\"dyn_HairA_BackR_03\"),Transform=(Rotation=(X=0,Y=0,Z=0,W=1),Translation=(X=-3,Y=0,Z=2),Scale3D=(X=1,Y=1,Z=1))),(Item=(Type=Bone,Name=\"dyn_HairA_BackL_03\"),Transform=(Rotation=(X=0,Y=0,Z=0,W=1),Translation=(X=-3,Y=0,Z=-2),Scale3D=(X=1 ...(truncated)"
    WeightMinimum=0
    Weight=@n4.Result
   )
   > &RigUnit_ModifyTransforms(
    ItemToModify="((Item=(Type=Bone,Name=\"dyn_HairA_Back_03\"),Transform=(Rotation=(X=0,Y=-0.0175,Z=0,W=0.9998),Translation=(X=0,Y=0,Z=0),Scale3D=(X=1,Y=1,Z=1))),(Item=(Type=Bone,Name=\"dyn_HairA_BackR_03\"),Transform=(Rotation=(X=0,Y=-0.0175,Z=0,W=0.9998),Translation=(X=0,Y=0 ...(truncated)"
    Weight=1
    WeightMinimum=0
   )
   > &RigUnit_ModifyTransforms(
    ItemToModify="((Item=(Type=Bone,Name=\"dyn_HairA_L_04\"),Transform=(Rotation=(X=0,Y=0,Z=0.0872,W=0.9962),Translation=(X=0,Y=0,Z=0),Scale3D=(X=1,Y=1,Z=1))),(Item=(Type=Bone,Name=\"dyn_HairA_mid_L_04\"),Transform=(Rotation=(X=0.0038,Y=-0.0435,Z=0.0871,W=0.9952),Translation=(X ...(truncated)"
    Weight=1
    WeightMinimum=0
   )
   > Get Delta Ref_1 = RigVMFunctionReferenceNode(Item=(Type=Bone,Name="head"))
   > RigVMVariableNode(Value=&RigVMFunction_MathDoubleRemap(SourceMaximum=50, SourceMinimum=0, TargetMaximum=1, TargetMinimum=0, Value=@"Get Delta Ref_1".Result.Rotation.Roll).Result) {variable=head rot x}
   > &RigUnit_ModifyTransforms(
    ItemToModify=((Item=(Type=Bone,Name="dyn_HairA_BackR_03"),Transform=(Rotation=(X=0,Y=0,Z=-0.1305,W=0.9914),Translation=(X=0,Y=0,Z=0),Scale3D=(X=1,Y=1,Z=1))))
    WeightMinimum=0
    Weight=RigVMVariableNode() {variable=head rot x}.Value
   )
   RigVMFunction_MathBoolToFloat = &RigVMFunction_MathBoolToFloat(Value=RigVMVariableNode() {variable=Skydivining}.Value)
   AlphaInterp = &RigVMFunction_AlphaInterp(
    InterpSpeedDecreasing=10
    InterpSpeedIncreasing=10
    Scale=1
    bClampResult=False
    Value=&RigVMFunction_MathDoubleMul(A=@RigVMFunction_MathBoolToFloat.Result, B=RigVMVariableNode() {variable=head rot z}.Value).Result
   )
   > &RigUnit_OffsetTransformForItem(
    Item=(Type=Bone,Name="dyn_HairA_BackL_02")
    OffsetTransform=(Rotation=(X=0,Y=0,Z=0,W=1),Translation=(X=0,Y=0,Z=5),Scale3D=(X=1,Y=1,Z=1))
    bPropagateToChildren=false
    Weight=@AlphaInterp.Result
   )
   > &RigUnit_OffsetTransformForItem(
    Item=(Type=Bone,Name="dyn_HairA_Back_02")
    OffsetTransform=(Rotation=(X=0,Y=0,Z=0,W=1),Translation=(X=0,Y=0,Z=4),Scale3D=(X=1,Y=1,Z=1))
    bPropagateToChildren=false
    Weight=@AlphaInterp.Result
   )
   > &RigUnit_OffsetTransformForItem(
    Item=(Type=Bone,Name="dyn_HairA_BackR_02")
    OffsetTransform=(Rotation=(X=0,Y=0,Z=0,W=1),Translation=(X=0,Y=0,Z=5),Scale3D=(X=1,Y=1,Z=1))
    bPropagateToChildren=false
    Weight=@AlphaInterp.Result
   )
  ]}
]
hierarchy [
 root {loc=<0 0 0>, rot=<0 0 0 1>}
 pelvis {parent=root, loc=<0 -0.5301 98.0571>, rot=<0.0127 0.707 -0.0127 -0.707>}
 spine_01 {parent=pelvis, loc=<6.1208 -0.9328 0>, rot=<0 0 0.034 -0.9994>}
 spine_02 {parent=spine_01, loc=<6.5997 -0.2095 0>, rot=<0 0 0.0238 -0.9997>}
 spine_03 {parent=spine_02, loc=<6.4914 -0.1237 0>, rot=<0 0 0.0149 -0.9999>}
 spine_04 {parent=spine_03, loc=<6.9598 0.0794 0>, rot=<0 0 -0.0132 -0.9999>}
 spine_05 {parent=spine_04, loc=<6.9987 0 0>, rot=<0 0 -0.0015 -1>}
 neck_01 {parent=spine_05, loc=<11.3138 0.1529 0>, rot=<0 0 -0.2003 -0.9797>}
 neck_02 {parent=neck_01, loc=<4.5627 0 0>, rot=<0 0 0.0504 -0.9987>}
 head {parent=neck_02, loc=<5.3914 0 0>, rot=<0 0 0.115 -0.9934>}
 base_HairA_BackR_01 {parent=head, loc=<6.1965 -3.2869 6.03>, rot=<0.0541 -0.4196 0.9044 -0.0558>}
 dyn_HairA_BackR_01 {parent=base_HairA_BackR_01, loc=<0 0 0>, rot=<0.0929 -0.0335 0.0702 -0.9926>}
 dyn_HairA_BackR_02 {parent=dyn_HairA_BackR_01, loc=<5.6131 0 0>, rot=<-0.0087 0.0592 -0.0531 -0.9968>}
 dyn_HairA_BackR_03 {parent=dyn_HairA_BackR_02, loc=<5.6131 0 0>, rot=<-0.0175 0.1591 -0.0919 -0.9828>}
 dyn_HairA_BackR_04 {parent=dyn_HairA_BackR_03, loc=<5.6131 0 0>, rot=<-0.0035 0.0307 -0.0499 -0.9983>}
 dyn_HairA_BackR_05 {parent=dyn_HairA_BackR_04, loc=<5.6131 0 0>, rot=<-0.0101 -0.2749 0.0744 -0.9585>}
 dyn_HairA_BackR_06 {parent=dyn_HairA_BackR_05, loc=<5.6131 0 0>, rot=<0 0 0 1>}
 base_HairA_Back_01 {parent=head, loc=<6.5329 -6.4927 0.8092>, rot=<0.0703 -0.7036 0.7036 -0.0703>}
 dyn_HairA_Back_01 {parent=base_HairA_Back_01, loc=<0 0 0>, rot=<0 0 0 1>}
 dyn_HairA_Back_02 {parent=dyn_HairA_Back_01, loc=<5.8294 0 0>, rot=<0 0.01 0.002 -0.9999>}
 dyn_HairA_Back_03 {parent=dyn_HairA_Back_02, loc=<5.8294 0 0>, rot=<-0.0009 0.0664 0.0091 -0.9978>}
 dyn_HairA_Back_04 {parent=dyn_HairA_Back_03, loc=<5.8294 0 0>, rot=<0.0018 -0.0682 0.0044 -0.9977>}
 dyn_HairA_Back_05 {parent=dyn_HairA_Back_04, loc=<5.8294 0 0>, rot=<0.0038 -0.0964 0.0081 -0.9953>}
 dyn_HairA_Back_06 {parent=dyn_HairA_Back_05, loc=<5.8294 0 0>, rot=<0.0045 -0.0836 0.0064 -0.9965>}
 dyn_HairA_Back_07 {parent=dyn_HairA_Back_06, loc=<5.8294 0 0>, rot=<0 0 0 1>}
 base_HairA_BackL_01 {parent=head, loc=<5.8103 -4.5167 -5.7556>, rot=<0.0536 -0.9377 0.3388 -0.0554>}
 dyn_HairA_BackL_01 {parent=base_HairA_BackL_01, loc=<0 0 0>, rot=<-0.175 -0.0147 -0.0138 -0.9844>}
 dyn_HairA_BackL_02 {parent=dyn_HairA_BackL_01, loc=<5.7229 0 0>, rot=<0.0053 0.0611 -0.007 -0.9981>}
 dyn_HairA_BackL_03 {parent=dyn_HairA_BackL_02, loc=<5.7229 0 0>, rot=<0.017 0.1025 0.0601 -0.9928>}
 dyn_HairA_BackL_04 {parent=dyn_HairA_BackL_03, loc=<5.7229 0 0>, rot=<0.0007 -0.0321 0.1166 -0.9927>}
 dyn_HairA_BackL_05 {parent=dyn_HairA_BackL_04, loc=<5.7229 0 0>, rot=<0.0326 -0.2076 -0.044 -0.9767>}
 dyn_HairA_BackL_06 {parent=dyn_HairA_BackL_05, loc=<5.723 0 0>, rot=<0 0 0 1>}
 base_HairA_front_L_01 {parent=head, loc=<1.5231 6.9804 -6.776>, rot=<0.1002 -0.9739 -0.1582 0.1282>}
 dyn_HairA_front_L_01 {parent=base_HairA_front_L_01, loc=<0 0 0>, rot=<0 0 0 1>}
 dyn_HairA_front_L_02 {parent=dyn_HairA_front_L_01, loc=<2.9883 0 0>, rot=<0.006 0.0954 -0.0123 -0.9953>}
 dyn_HairA_front_L_03 {parent=dyn_HairA_front_L_02, loc=<2.9796 0 0>, rot=<-0.01 0.1657 -0.2051 -0.9646>}
 dyn_HairA_front_L_04 {parent=dyn_HairA_front_L_03, loc=<3.0054 0 0>, rot=<0 0 0 1>}
 base_HairA_L_01 {parent=head, loc=<2.958 -0.0365 -7.9103>, rot=<0.0795 -0.9955 -0.004 -0.0505>}
 dyn_HairA_L_01 {parent=base_HairA_L_01, loc=<0 0 0>, rot=<0 0 0 1>}
 dyn_HairA_L_02 {parent=dyn_HairA_L_01, loc=<3.5871 0 0>, rot=<-0.0047 0.031 -0.0352 -0.9989>}
 dyn_HairA_L_03 {parent=dyn_HairA_L_02, loc=<3.5669 0 0>, rot=<-0.0095 -0.0209 -0.066 -0.9976>}
 dyn_HairA_L_04 {parent=dyn_HairA_L_03, loc=<3.6702 0 0>, rot=<-0.0217 0.0029 -0.1734 -0.9846>}
 dyn_HairA_L_05 {parent=dyn_HairA_L_04, loc=<3.4914 0 0>, rot=<-0.04 0.0405 -0.2344 -0.9705>}
 dyn_HairA_L_06 {parent=dyn_HairA_L_05, loc=<3.6419 0 0>, rot=<0 0 0 1>}
 base_HairA_R_01 {parent=head, loc=<4.0974 1.6889 7.7932>, rot=<-0.0597 0.1736 0.9829 0.0156>}
 dyn_HairA_R_01 {parent=base_HairA_R_01, loc=<0 0 0>, rot=<0 0 0 1>}
 dyn_HairA_R_02 {parent=dyn_HairA_R_01, loc=<4.1888 0 0>, rot=<0.0046 -0.0065 -0.0397 -0.9992>}
 dyn_HairA_R_03 {parent=dyn_HairA_R_02, loc=<4.1888 0 0>, rot=<-0.009 -0.0004 0.0716 -0.9974>}
 dyn_HairA_R_04 {parent=dyn_HairA_R_03, loc=<4.1888 0 0>, rot=<-0.0224 0.0246 0.1239 -0.9917>}
 dyn_HairA_R_05 {parent=dyn_HairA_R_04, loc=<4.1888 0 0>, rot=<-0.0145 0.2124 0.1913 -0.9582>}
 dyn_HairA_R_06 {parent=dyn_HairA_R_05, loc=<4.1888 0 0>, rot=<0 0 0 1>}
 dyn_HairA_B_R_01 {parent=dyn_HairA_R_03, loc=<1.187 2.5609 0.6836>, rot=<-0.0875 0.0817 -0.0276 -0.9924>}
 dyn_HairA_B_R_02 {parent=dyn_HairA_B_R_01, loc=<3.053 0 0>, rot=<-0.0102 0.0339 0.1942 -0.9803>}
 dyn_HairA_B_R_03 {parent=dyn_HairA_B_R_02, loc=<3.0479 0 0>, rot=<-0.0298 0.0403 0.1709 -0.984>}
 dyn_HairA_B_R_04 {parent=dyn_HairA_B_R_03, loc=<3.0539 0 0>, rot=<0 0 0 1>}
 dyn_HairA_A_R_01 {parent=dyn_HairA_R_03, loc=<0.4159 -1.8399 1.2711>, rot=<0.1438 0.173 0.1054 -0.9686>}
 dyn_HairA_A_R_02 {parent=dyn_HairA_A_R_01, loc=<2.7423 0 0>, rot=<0.0077 -0.054 0.1315 -0.9898>}
 dyn_HairA_A_R_03 {parent=dyn_HairA_A_R_02, loc=<2.8637 0 0>, rot=<0.097 -0.3671 -0.1681 -0.9097>}
 dyn_HairA_A_R_04 {parent=dyn_HairA_A_R_03, loc=<2.6204 0 0>, rot=<0 0 0 1>}
 dyn_HairA_front_uppr_A_R_01 {parent=dyn_HairA_R_02, loc=<1.7165 -0.4057 1.7847>, rot=<0.1518 0.1726 -0.5859 -0.7771>}
 dyn_HairA_front_uppr_A_R_02 {parent=dyn_HairA_front_uppr_A_R_01, loc=<3.0678 0 0>, rot=<-0.0535 0 -0.129 -0.9902>}
 dyn_HairA_front_uppr_A_R_03 {parent=dyn_HairA_front_uppr_A_R_02, loc=<3.0433 0 0>, rot=<0 0 0 1>}
 base_HairA_mid_L_01 {parent=head, loc=<2.3297 3.1813 -7.2358>, rot=<0.1246 -0.9869 -0.0928 -0.0428>}
 dyn_HairA_mid_L_01 {parent=base_HairA_mid_L_01, loc=<0 0 0>, rot=<0 0 0 1>}
 dyn_HairA_mid_L_02 {parent=dyn_HairA_mid_L_01, loc=<4.0333 0 0>, rot=<-0.0002 -0.0091 -0.0191 -0.9998>}
 dyn_HairA_mid_L_03 {parent=dyn_HairA_mid_L_02, loc=<4.0333 0 0>, rot=<0.0001 0.0009 -0.1175 -0.9931>}
 dyn_HairA_mid_L_04 {parent=dyn_HairA_mid_L_03, loc=<4.0332 0 0>, rot=<-0.0372 -0.1771 -0.2022 -0.9625>}
 dyn_HairA_mid_L_05 {parent=dyn_HairA_mid_L_04, loc=<4.0334 0 0>, rot=<-0.0339 -0.1873 -0.1746 -0.9661>}
 dyn_HairA_mid_L_06 {parent=dyn_HairA_mid_L_05, loc=<4.0333 0 0>, rot=<0 0 0 1>}
 head_hair_back_collision {parent=head, loc=<0 0 0>, rot=<0 0 0 1>}
 clavicle_l {parent=spine_05, loc=<5.4831 -0.1053 -3.2946>, rot=<0.0465 -0.7289 -0.009 -0.6829>}
 clavicle_l__hair_back_collision {parent=clavicle_l, loc=<0 0 0>, rot=<0 0 0 1>}
 upperarm_l {parent=clavicle_l, loc=<9.8658 -1.2136 -0.4105>, rot=<0.0164 -0.3245 -0.0269 -0.9454>}
 clavicle_r {parent=spine_05, loc=<5.4831 -0.1053 3.2946>, rot=<0.7289 0.0465 -0.6829 0.009>}
 clavicle_r__hair_back_collision {parent=clavicle_r, loc=<0 0 0>, rot=<0 0 0 1>}
 upperarm_r {parent=clavicle_r, loc=<-9.8658 1.2136 0.4105>, rot=<0.0164 -0.3245 -0.0269 -0.9454>}
 spine_04__hair_back_collision {parent=spine_04, loc=<0 0 0>, rot=<0 0 0 1>}
 spine_03_ctrl {controlType=EULER_TRANSFORM, type=Control, loc=<0 0 0>, rot=<0 0 0 1>}
 spine_04_ctrl {controlType=EULER_TRANSFORM, parent=spine_03_ctrl, type=Control, loc=<0 0 0>, rot=<0 0 0 1>}
 spine_05_ctrl {controlType=EULER_TRANSFORM, parent=spine_04_ctrl, type=Control, loc=<0 0 0>, rot=<0 0 0 1>}
 neck_01_ctrl {controlType=EULER_TRANSFORM, parent=spine_05_ctrl, type=Control, loc=<0 0 0>, rot=<0 0 0 1>}
 neck_02_ctrl {controlType=EULER_TRANSFORM, parent=neck_01_ctrl, type=Control, loc=<0 0 0>, rot=<0 0 0 1>}
 head_ctrl {controlType=EULER_TRANSFORM, parent=neck_02_ctrl, type=Control, loc=<0 0 0>, rot=<0 0 0 1>}
 clavicle_l_ctrl {controlType=EULER_TRANSFORM, parent=spine_05_ctrl, type=Control, loc=<0 0 0>, rot=<0 0 0 1>}
 clavicle_r_ctrl {controlType=EULER_TRANSFORM, parent=spine_05_ctrl, type=Control, loc=<0 0 0>, rot=<0 0 0 1>}
]
variables [
 BackBiling {type=bool, default=False}
 Skydivining {type=bool, default=False}
 head rot x {type=double, default=0}
 head rot z {type=double, default=0}
 is Targeting {type=bool, default=True}
]
