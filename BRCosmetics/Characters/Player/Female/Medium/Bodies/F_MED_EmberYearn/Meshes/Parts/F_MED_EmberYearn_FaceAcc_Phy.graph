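# Physics asset for the EmberYearn female character's FaceAcc part, setting up hair strand dynamics with upper-body collision shapes.
# Capsules cover head, neck, upper spine, clavicles and upper arms, with all joints locked except the free shoulders.
# Zero-size base bodies lock four hair roots to the head, feeding capsule chains with tight angular limits.
# The right chain branches into sub-strands, and an angularly locked cross-link ties the left and mid-left chains together.
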PhysicsAsset /BRCosmetics/Characters/Player/Female/Medium/Bodies/F_MED_EmberYearn/Meshes/Parts/F_MED_EmberYearn_FaceAcc_Phy
bodies {
 base_HairA_L_01 [{shapeType=Capsule, length=0, radius=0, shapeName=base_HairA_L_01_capsule}]
 base_HairA_R_01 [{shapeType=Capsule, length=0, radius=0, shapeName=base_HairA_R_01_capsule}]
 base_HairA_front_L_01 [{shapeType=Capsule, length=0, radius=0, shapeName=base_HairA_front_L_01_capsule}]
 base_HairA_mid_L_01 [{shapeType=Capsule, length=0, radius=0, shapeName=base_HairA_mid_L_01_capsule}]
 clavicle_l [{
   shapeType=Capsule
   length=13.1368
   radius=7.3
   shapeName=clavicle_l_capsule
   center=<4.8324 -0.5907 -2.7596>
   rotation=<rot 85.5566 -2.6368 -4.9503>
  }]
 clavicle_r [{
   shapeType=Capsule
   length=8
   radius=7.2
   shapeName=clavicle_r_capsule
   center=<-4.8351 1.223 2.5543>
   rotation=<rot -80.128 -179.7948 177.3409>
  }]
 dyn_HairA_A_R_01 [{shapeType=Capsule, length=1, radius=1.5, shapeName=dyn_HairA_A_R_01_capsule, center=<1.6767 0 0>, rotation=<rot 90 0 0>}]
 dyn_HairA_A_R_02 [{shapeType=Capsule, length=1.5, radius=1, shapeName=dyn_HairA_A_R_02_capsule, center=<1.6767 0 0>, rotation=<rot 90 0 0>}]
 dyn_HairA_A_R_03 [{shapeType=Capsule, length=2, radius=0.5, shapeName=dyn_HairA_A_R_03_capsule, center=<1.5653 -0.0509 0.2197>, rotation=<rot 90 22.1156 0>}]
 dyn_HairA_B_R_01 [{shapeType=Capsule, length=0.5, radius=2, shapeName=dyn_HairA_B_R_01_capsule, center=<1.6767 0.2752 0>, rotation=<rot 90 0 0>}]
 dyn_HairA_B_R_02 [{shapeType=Capsule, length=0.5, radius=2, shapeName=dyn_HairA_B_R_02_capsule, center=<1.6767 -0.2661 0>, rotation=<rot 90 0 0>}]
 dyn_HairA_B_R_03 [{shapeType=Capsule, length=2, radius=1, shapeName=dyn_HairA_B_R_03_capsule, center=<1.1758 0.1375 0>, rotation=<rot 90 -15.3434 0>}]
 dyn_HairA_L_01 [{
   shapeType=Capsule
   length=1.8
   radius=1.7
   shapeName=dyn_HairA_L_01_capsule
   center=<1.8905 0 0.4166>
   rotation=<rot 87.1415 -0.0001 -0.0001>
  }]
 dyn_HairA_L_02 [{shapeType=Capsule, length=1.8, radius=1.7, shapeName=dyn_HairA_L_02_capsule, center=<1.8905 0 0.0869>, rotation=<rot 85.85 -0.0001 -0.0002>}]
 dyn_HairA_L_03 [{shapeType=Capsule, length=1.6, radius=1.9, shapeName=dyn_HairA_L_03_capsule, center=<1.8905 0 0>, rotation=<rot 90 0 0>}]
 dyn_HairA_L_04 [{shapeType=Capsule, length=1.9, radius=1.6, shapeName=dyn_HairA_L_04_capsule, center=<1.8905 0 0>, rotation=<rot 90 0 0>}]
 dyn_HairA_L_05 [{shapeType=Capsule, length=2, radius=1.5, shapeName=dyn_HairA_L_05_capsule, center=<1.4824 0.1448 0>, rotation=<rot 90 5.7824 0>}]
 dyn_HairA_R_01 [{shapeType=Capsule, length=2.1, radius=1.4, shapeName=dyn_HairA_R_01_capsule, center=<1.7473 0 0.4001>, rotation=<rot 85.1977 -180 -179.9999>}]
 dyn_HairA_R_02 [{shapeType=Capsule, length=2.2, radius=1.3, shapeName=dyn_HairA_R_02_capsule, center=<1.6767 0 0.7202>, rotation=<rot 90 0 0>}]
 dyn_HairA_R_03 [{shapeType=Capsule, length=1.8, radius=1.7, shapeName=dyn_HairA_R_03_capsule, center=<2.0697 0 0.5278>, rotation=<rot 84.8103 -180 -179.9999>}]
 dyn_HairA_R_04 [{shapeType=Capsule, length=1.5, radius=2, shapeName=dyn_HairA_R_04_capsule, center=<2.1267 0 0.3628>, rotation=<rot 90 0 0>}]
 dyn_HairA_R_05 [{shapeType=Capsule, length=3.7, radius=1, shapeName=dyn_HairA_R_05_capsule, center=<1.4012 -0.2078 -0.5997>, rotation=<rot 90 -3.8361 0>}]
 dyn_HairA_front_L_01 [{shapeType=Capsule, length=1.5, radius=1, shapeName=dyn_HairA_front_L_01_capsule, center=<1.8905 0 0.3011>, rotation=<rot 90 0 0>}]
 dyn_HairA_front_L_02 [{shapeType=Capsule, length=1.5, radius=1, shapeName=dyn_HairA_front_L_02_capsule, center=<1.8905 0.2018 0.3628>, rotation=<rot 90 0 0>}]
 dyn_HairA_front_L_03 [{
   shapeType=Capsule
   length=1
   radius=1
   shapeName=dyn_HairA_front_L_03_capsule
   center=<1.2715 -0.0827 0.2074>
   rotation=<rot 84.7081 -9.4278 -0.0001>
  }]
 dyn_HairA_front_uppr_A_R_01 [{shapeType=Capsule, length=1.5, radius=1, shapeName=dyn_HairA_front_uppr_A_R_01_capsule, center=<1.6767 0 0>, rotation=<rot 90 0 0>}]
 dyn_HairA_front_uppr_A_R_02 [{
   shapeType=Capsule
   length=2
   radius=0.5
   shapeName=dyn_HairA_front_uppr_A_R_02_capsule
   center=<1.5799 -0.0138 0>
   rotation=<rot 90 8.1179 0>
  }]
 dyn_HairA_mid_L_01 [{shapeType=Capsule, length=1.5, radius=1.5, shapeName=dyn_HairA_mid_L_01_capsule, center=<2.0407 0 1.2168>, rotation=<rot 82.9633 0 -0.0001>}]
 dyn_HairA_mid_L_02 [{shapeType=Capsule, length=1.3, radius=1.7, shapeName=dyn_HairA_mid_L_02_capsule, center=<1.9385 0 0.6859>, rotation=<rot 85.9942 0 -0.0001>}]
 dyn_HairA_mid_L_03 [{shapeType=Capsule, length=1, radius=2, shapeName=dyn_HairA_mid_L_03_capsule, center=<1.8905 0 0>, rotation=<rot 90 0 0>}]
 dyn_HairA_mid_L_04 [{shapeType=Capsule, length=1, radius=2, shapeName=dyn_HairA_mid_L_04_capsule, center=<1.4487 -0.1743 0>, rotation=<rot 90 23.1334 0>}]
 dyn_HairA_mid_L_05 [{shapeType=Capsule, length=2.5, radius=1.5, shapeName=dyn_HairA_mid_L_05_capsule, center=<3 1 0>, rotation=<rot 90 -7.4049 0>}]
 head [{shapeType=Capsule, length=5.0255, radius=6.5197, shapeName=head_capsule, center=<1.6421 2.2496 0>, rotation=<rot 90 7.805 0>}]
 neck_01 [{shapeType=Capsule, length=3.5, radius=4.505, shapeName=neck_01_capsule, center=<1.7767 2.1443 0>, rotation=<rot 90 -2.4224 0>}]
 spine_05 [{shapeType=Capsule, length=6.5, radius=15.405, shapeName=spine_05_capsule, center=<-5.1683 -4.5021 0>, rotation=<rot 0 -90 0>}]
 upperarm_l [{shapeType=Capsule, length=16.8, radius=6, shapeName=upperarm_l_capsule, center=<11.4746 -0.5285 -0.899>, rotation=<rot -90 0 0>}]
 upperarm_r [{shapeType=Capsule, length=16.7, radius=5.305, shapeName=upperarm_r_capsule, center=<-11.4746 0.3698 0.5382>, rotation=<rot 90 0 0>}]
}
constraints [
 {
  bone1Name=clavicle_l
  bone2Name=spine_05
  swing1LimitDegrees=45
  swing1Motion=Locked
  swing2LimitDegrees=45
  swing2Motion=Locked
  twistLimitDegrees=45
  twistMotion=Locked
 }
 {
  bone1Name=clavicle_r
  bone2Name=spine_05
  swing1LimitDegrees=45
  swing1Motion=Locked
  swing2LimitDegrees=45
  swing2Motion=Locked
  twistLimitDegrees=45
  twistMotion=Locked
 }
 {
  bone1Name=neck_01
  bone2Name=spine_05
  swing1LimitDegrees=45
  swing1Motion=Locked
  swing2LimitDegrees=45
  swing2Motion=Locked
  twistLimitDegrees=45
  twistMotion=Locked
 }
 {
  bone1Name=head
  bone2Name=neck_01
  swing1LimitDegrees=45
  swing1Motion=Locked
  swing2LimitDegrees=45
  swing2Motion=Locked
  twistLimitDegrees=45
  twistMotion=Locked
 }
 {
  bone1Name=base_HairA_L_01
  bone2Name=head
  swing1LimitDegrees=45
  swing1Motion=Locked
  swing2LimitDegrees=45
  swing2Motion=Locked
  twistLimitDegrees=45
  twistMotion=Locked
 }
 {bone1Name=dyn_HairA_L_01, bone2Name=base_HairA_L_01, swing1LimitDegrees=10, swing2LimitDegrees=5, twistLimitDegrees=2}
 {bone1Name=dyn_HairA_L_02, bone2Name=dyn_HairA_L_01, swing1LimitDegrees=10, swing2LimitDegrees=10, twistLimitDegrees=2}
 {bone1Name=dyn_HairA_L_03, bone2Name=dyn_HairA_L_02, swing1LimitDegrees=15, swing2LimitDegrees=15, twistLimitDegrees=2}
 {bone1Name=dyn_HairA_L_04, bone2Name=dyn_HairA_L_03, swing1LimitDegrees=15, swing2LimitDegrees=15, twistLimitDegrees=2}
 {bone1Name=dyn_HairA_L_05, bone2Name=dyn_HairA_L_04, swing1LimitDegrees=15, swing2LimitDegrees=15, twistLimitDegrees=2}
 {
  bone1Name=base_HairA_R_01
  bone2Name=head
  swing1LimitDegrees=45
  swing1Motion=Locked
  swing2LimitDegrees=45
  swing2Motion=Locked
  twistLimitDegrees=45
  twistMotion=Locked
 }
 {bone1Name=dyn_HairA_R_01, bone2Name=base_HairA_R_01, swing1LimitDegrees=10, swing2LimitDegrees=5, twistLimitDegrees=2}
 {bone1Name=dyn_HairA_R_02, bone2Name=dyn_HairA_R_01, swing1LimitDegrees=10, swing2LimitDegrees=10, twistLimitDegrees=2}
 {bone1Name=dyn_HairA_R_03, bone2Name=dyn_HairA_R_02, swing1LimitDegrees=10, swing2LimitDegrees=15, twistLimitDegrees=2}
 {bone1Name=dyn_HairA_A_R_01, bone2Name=dyn_HairA_R_03, swing1LimitDegrees=10, swing2LimitDegrees=15, twistLimitDegrees=2}
 {bone1Name=dyn_HairA_A_R_02, bone2Name=dyn_HairA_A_R_01, swing1LimitDegrees=10, swing2LimitDegrees=15, twistLimitDegrees=2}
 {bone1Name=dyn_HairA_A_R_03, bone2Name=dyn_HairA_A_R_02, swing1LimitDegrees=10, swing2LimitDegrees=15, twistLimitDegrees=2}
 {bone1Name=dyn_HairA_B_R_01, bone2Name=dyn_HairA_R_03, swing1LimitDegrees=10, swing2LimitDegrees=15, twistLimitDegrees=2}
 {bone1Name=dyn_HairA_B_R_02, bone2Name=dyn_HairA_B_R_01, swing1LimitDegrees=10, swing2LimitDegrees=15, twistLimitDegrees=2}
 {bone1Name=dyn_HairA_B_R_03, bone2Name=dyn_HairA_B_R_02, swing1LimitDegrees=10, swing2LimitDegrees=15, twistLimitDegrees=2}
 {bone1Name=dyn_HairA_R_04, bone2Name=dyn_HairA_R_03, swing1LimitDegrees=10, swing2LimitDegrees=15, twistLimitDegrees=2}
 {bone1Name=dyn_HairA_R_05, bone2Name=dyn_HairA_R_04, swing1LimitDegrees=10, swing2LimitDegrees=15, twistLimitDegrees=2}
 {bone1Name=dyn_HairA_front_uppr_A_R_01, bone2Name=dyn_HairA_R_02, swing1LimitDegrees=2, swing2LimitDegrees=5, twistLimitDegrees=2}
 {
  bone1Name=dyn_HairA_front_uppr_A_R_02
  bone2Name=dyn_HairA_front_uppr_A_R_01
  swing1LimitDegrees=2
  swing2LimitDegrees=5
  twistLimitDegrees=2
 }
 {
  bone1Name=base_HairA_front_L_01
  bone2Name=head
  swing1LimitDegrees=45
  swing1Motion=Locked
  swing2LimitDegrees=45
  swing2Motion=Locked
  twistLimitDegrees=45
  twistMotion=Locked
 }
 {bone1Name=dyn_HairA_front_L_01, bone2Name=base_HairA_front_L_01, swing1LimitDegrees=10, swing2LimitDegrees=10, twistLimitDegrees=2}
 {bone1Name=dyn_HairA_front_L_02, bone2Name=dyn_HairA_front_L_01, swing1LimitDegrees=15, swing2LimitDegrees=15, twistLimitDegrees=2}
 {bone1Name=dyn_HairA_front_L_03, bone2Name=dyn_HairA_front_L_02, swing1LimitDegrees=15, swing2LimitDegrees=15, twistLimitDegrees=2}
 {
  bone1Name=base_HairA_mid_L_01
  bone2Name=head
  swing1LimitDegrees=45
  swing1Motion=Locked
  swing2LimitDegrees=45
  swing2Motion=Locked
  twistLimitDegrees=45
  twistMotion=Locked
 }
 {bone1Name=dyn_HairA_mid_L_01, bone2Name=base_HairA_mid_L_01, swing1LimitDegrees=10, swing2LimitDegrees=5, twistLimitDegrees=2}
 {bone1Name=dyn_HairA_mid_L_02, bone2Name=dyn_HairA_mid_L_01, swing1LimitDegrees=10, swing2LimitDegrees=10, twistLimitDegrees=2}
 {bone1Name=dyn_HairA_mid_L_03, bone2Name=dyn_HairA_mid_L_02, swing1LimitDegrees=15, swing2LimitDegrees=15, twistLimitDegrees=2}
 {bone1Name=dyn_HairA_mid_L_04, bone2Name=dyn_HairA_mid_L_03, swing1LimitDegrees=15, swing2LimitDegrees=15, twistLimitDegrees=2}
 {bone1Name=dyn_HairA_mid_L_05, bone2Name=dyn_HairA_mid_L_04, swing1LimitDegrees=15, swing2LimitDegrees=15, twistLimitDegrees=2}
 {
  bone1Name=upperarm_r
  bone2Name=clavicle_r
  swing1LimitDegrees=45
  swing1Motion=Free
  swing2LimitDegrees=45
  swing2Motion=Free
  twistLimitDegrees=45
  twistMotion=Free
 }
 {
  bone1Name=upperarm_l
  bone2Name=clavicle_l
  swing1LimitDegrees=45
  swing1Motion=Free
  swing2LimitDegrees=45
  swing2Motion=Free
  twistLimitDegrees=45
  twistMotion=Free
 }
 {
  bone1Name=dyn_HairA_mid_L_03
  bone2Name=dyn_HairA_L_03
  swing1LimitDegrees=45
  swing1Motion=Locked
  swing2LimitDegrees=45
  swing2Motion=Locked
  twistLimitDegrees=45
  twistMotion=Locked
 }
]
