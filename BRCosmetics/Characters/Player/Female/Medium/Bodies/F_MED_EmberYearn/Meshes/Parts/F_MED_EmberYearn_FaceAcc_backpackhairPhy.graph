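# Physics asset for the back hair of the EmberYearn female character's face accessory, with collision for a backpack.
# Center, left and right capsule chains start at zero-size base bodies and link segment to segment with tight angular limits.
# Locked cross-constraints tie both side chains to the center chain at their third segments.
# Head, upper spine, clavicle, upper arm and two-capsule backpack bodies supply the surrounding body shapes.
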PhysicsAsset /BRCosmetics/Characters/Player/Female/Medium/Bodies/F_MED_EmberYearn/Meshes/Parts/F_MED_EmberYearn_FaceAcc_backpackhairPhy
bodies {
 attach_backpack [
  {
   shapeType=Capsule
   length=30
   radius=20
   shapeName=attach_backpack_capsule
   center=<1.7822 -22.245 -1.7141>
   rotation=<rot 83.138 167.4531 165.4396>
  }
  {shapeType=Capsule, length=15, radius=15, shapeName=attach_backpack_capsule, center=<-3.0349 -8.1778 27.4782>, rotation=<rot 0 0 24.4897>}
 ]
 base_HairA_BackL_01 [{shapeType=Capsule, length=0, radius=0, shapeName=base_HairA_BackL_01_capsule}]
 base_HairA_BackR_01 [{shapeType=Capsule, length=0, radius=0, shapeName=base_HairA_BackR_01_capsule}]
 base_HairA_Back_01 [{shapeType=Capsule, length=0, radius=0, shapeName=base_HairA_Back_01_capsule}]
 clavicle_l [{
   shapeType=Capsule
   length=6.7
   radius=7.3
   shapeName=clavicle_l_capsule
   center=<4.8513 -0.4117 -2.3994>
   rotation=<rot 80.55 -0.0703 -3.0836>
  }]
 clavicle_r [{
   shapeType=Capsule
   length=6.7
   radius=7.3
   shapeName=clavicle_r_capsule
   center=<-4.8513 0.4117 2.3994>
   rotation=<rot -80.5505 179.9972 -175.0984>
  }]
 dyn_HairA_BackL_01 [{
   shapeType=Capsule
   length=2.5
   radius=2.5
   shapeName=dyn_HairA_BackL_01_capsule
   center=<2.0815 0.0473 0.6562>
   rotation=<rot 76.56 89.9992 179.999>
  }]
 dyn_HairA_BackL_02 [{
   shapeType=Capsule
   length=2.7
   radius=2.3
   shapeName=dyn_HairA_BackL_02_capsule
   center=<2.0815 -0.0352 -0.1812>
   rotation=<rot 78.994 89.9993 179.9992>
  }]
 dyn_HairA_BackL_03 [{
   shapeType=Capsule
   length=2.6
   radius=2.9
   shapeName=dyn_HairA_BackL_03_capsule
   center=<2.0815 0.0162 -0.6111>
   rotation=<rot 88.4788 -89.9953 0.0046>
  }]
 dyn_HairA_BackL_04 [{
   shapeType=Capsule
   length=2.9
   radius=2.6
   shapeName=dyn_HairA_BackL_04_capsule
   center=<2.0821 -0.5861 1.5407>
   rotation=<rot 69.457 -89.9457 0.0012>
  }]
 dyn_HairA_BackL_05 [{
   shapeType=Capsule
   length=3.8
   radius=3.5
   shapeName=dyn_HairA_BackL_05_capsule
   center=<0.5406 -1.6357 1.9953>
   rotation=<rot 82.6813 179.9999 179.9999>
  }]
 dyn_HairA_BackR_01 [{
   shapeType=Capsule
   length=3
   radius=2
   shapeName=dyn_HairA_BackR_01_capsule
   center=<2.6705 -0.6378 1.0916>
   rotation=<rot 53.8265 -89.9951 0.0104>
  }]
 dyn_HairA_BackR_02 [{
   shapeType=Capsule
   length=3.3
   radius=2.2
   shapeName=dyn_HairA_BackR_02_capsule
   center=<2.0815 -0.2272 0.5084>
   rotation=<rot 65.9416 -89.9898 0.015>
  }]
 dyn_HairA_BackR_03 [{
   shapeType=Capsule
   length=2.4
   radius=3.1
   shapeName=dyn_HairA_BackR_03_capsule
   center=<2.3851 0.1352 0.472>
   rotation=<rot 73.1052 -89.9832 0.0212>
  }]
 dyn_HairA_BackR_04 [{
   shapeType=Capsule
   length=2.2
   radius=3.5
   shapeName=dyn_HairA_BackR_04_capsule
   center=<2.0818 0.9008 2.6666>
   rotation=<rot 83.8611 -89.9475 0.0563>
  }]
 dyn_HairA_BackR_05 [{
   shapeType=Capsule
   length=2.8
   radius=3
   shapeName=dyn_HairA_BackR_05_capsule
   center=<2.5812 -0.1991 3.3437>
   rotation=<rot 65.8525 107.9461 15.161>
  }]
 dyn_HairA_Back_01 [{shapeType=Capsule, length=7.2, radius=2, shapeName=dyn_HairA_Back_01_capsule, center=<2.0815 0 0.976>, rotation=<rot 90 -90 0>}]
 dyn_HairA_Back_02 [{shapeType=Capsule, length=6.2, radius=3, shapeName=dyn_HairA_Back_02_capsule, center=<2.0815 0 0.6938>, rotation=<rot 90 -90 0>}]
 dyn_HairA_Back_03 [{shapeType=Capsule, length=5.2, radius=4, shapeName=dyn_HairA_Back_03_capsule, center=<2.0815 0 0>, rotation=<rot 90 -90 0>}]
 dyn_HairA_Back_04 [{shapeType=Capsule, length=7.8, radius=3.5, shapeName=dyn_HairA_Back_04_capsule, center=<2.0815 0 1.5762>, rotation=<rot 90 -90 0>}]
 dyn_HairA_Back_05 [{shapeType=Capsule, length=7.8, radius=3.5, shapeName=dyn_HairA_Back_05_capsule, center=<2.0815 0 2.6857>, rotation=<rot 90 -90 0>}]
 dyn_HairA_Back_06 [{
   shapeType=Capsule
   length=5.6
   radius=2.6
   shapeName=dyn_HairA_Back_06_capsule
   center=<1.617 0 1.7492>
   rotation=<rot 89.5366 179.9421 179.9421>
  }]
 head [{shapeType=Capsule, length=4.5255, radius=7.0197, shapeName=head_capsule, center=<-0.1656 2.0019 0>, rotation=<rot 90 7.805 0>}]
 spine_04 [{shapeType=Capsule, length=16.2, radius=14.105, shapeName=spine_04_capsule, center=<5.0288 4.8958 0>, rotation=<rot 0 -90 0>}]
 upperarm_l [{shapeType=Capsule, length=16.5, radius=5.505, shapeName=upperarm_l_capsule, center=<11.4746 0 0>, rotation=<rot -90 0 0>}]
 upperarm_r [{shapeType=Capsule, length=16.5, radius=5.505, shapeName=upperarm_r_capsule, center=<-11.4746 0 0>, rotation=<rot 90 0 0>}]
}
constraints [
 {
  bone1Name=clavicle_l
  bone2Name=spine_04
  swing1LimitDegrees=45
  swing1Motion=Locked
  swing2LimitDegrees=45
  swing2Motion=Locked
  twistLimitDegrees=45
  twistMotion=Locked
 }
 {
  bone1Name=clavicle_r
  bone2Name=spine_04
  swing1LimitDegrees=45
  swing1Motion=Locked
  swing2LimitDegrees=45
  swing2Motion=Locked
  twistLimitDegrees=45
  twistMotion=Locked
 }
 {
  bone1Name=head
  bone2Name=spine_04
  swing1LimitDegrees=45
  swing1Motion=Locked
  swing2LimitDegrees=45
  swing2Motion=Locked
  twistLimitDegrees=45
  twistMotion=Locked
 }
 {bone1Name=dyn_HairA_BackL_01, bone2Name=base_HairA_BackL_01, swing1LimitDegrees=2, swing2LimitDegrees=2, twistLimitDegrees=2}
 {bone1Name=dyn_HairA_BackR_01, bone2Name=base_HairA_BackR_01, swing1LimitDegrees=2, swing2LimitDegrees=2, twistLimitDegrees=2}
 {bone1Name=dyn_HairA_Back_01, bone2Name=base_HairA_Back_01, swing1LimitDegrees=2, swing2LimitDegrees=2, twistLimitDegrees=2}
 {bone1Name=dyn_HairA_Back_02, bone2Name=dyn_HairA_Back_01, swing1LimitDegrees=2, swing2LimitDegrees=2, twistLimitDegrees=2}
 {bone1Name=dyn_HairA_Back_03, bone2Name=dyn_HairA_Back_02, swing1LimitDegrees=2, swing2LimitDegrees=2, twistLimitDegrees=2}
 {bone1Name=dyn_HairA_Back_04, bone2Name=dyn_HairA_Back_03, swing1LimitDegrees=2, swing2LimitDegrees=2, twistLimitDegrees=2}
 {bone1Name=dyn_HairA_BackL_02, bone2Name=dyn_HairA_BackL_01, swing1LimitDegrees=2, swing2LimitDegrees=2, twistLimitDegrees=2}
 {bone1Name=dyn_HairA_BackL_03, bone2Name=dyn_HairA_BackL_02, swing1LimitDegrees=2, swing2LimitDegrees=2, twistLimitDegrees=2}
 {bone1Name=dyn_HairA_BackL_04, bone2Name=dyn_HairA_BackL_03, swing1LimitDegrees=2, swing2LimitDegrees=2, twistLimitDegrees=2}
 {bone1Name=dyn_HairA_BackL_05, bone2Name=dyn_HairA_BackL_04, swing1LimitDegrees=2, swing2LimitDegrees=2, twistLimitDegrees=2}
 {bone1Name=dyn_HairA_BackR_02, bone2Name=dyn_HairA_BackR_01, swing1LimitDegrees=2, swing2LimitDegrees=2, twistLimitDegrees=2}
 {bone1Name=dyn_HairA_BackR_03, bone2Name=dyn_HairA_BackR_02, swing1LimitDegrees=2, swing2LimitDegrees=2, twistLimitDegrees=2}
 {bone1Name=dyn_HairA_BackR_04, bone2Name=dyn_HairA_BackR_03, swing1LimitDegrees=2, swing2LimitDegrees=2, twistLimitDegrees=2}
 {bone1Name=dyn_HairA_BackR_05, bone2Name=dyn_HairA_BackR_04, swing1LimitDegrees=2, swing2LimitDegrees=2, twistLimitDegrees=2}
 {bone1Name=dyn_HairA_Back_05, bone2Name=dyn_HairA_Back_04, swing1LimitDegrees=2, swing2LimitDegrees=2, twistLimitDegrees=2}
 {bone1Name=dyn_HairA_Back_06, bone2Name=dyn_HairA_Back_05, swing1LimitDegrees=2, swing2LimitDegrees=2, twistLimitDegrees=2}
 {
  bone1Name=upperarm_r
  bone2Name=clavicle_r
  swing1LimitDegrees=45
  swing1Motion=Free
  swing2LimitDegrees=45
  swing2Motion=Free
  twistLimitDegrees=45
  twistMotion=Free
 }
 {
  bone1Name=upperarm_l
  bone2Name=clavicle_l
  swing1LimitDegrees=45
  swing1Motion=Free
  swing2LimitDegrees=45
  swing2Motion=Free
  twistLimitDegrees=45
  twistMotion=Free
 }
 {
  bone1Name=dyn_HairA_Back_03
  bone2Name=dyn_HairA_BackL_03
  swing1LimitDegrees=45
  swing1Motion=Locked
  swing2LimitDegrees=45
  swing2Motion=Locked
  twistLimitDegrees=45
  twistMotion=Locked
 }
 {
  bone1Name=dyn_HairA_Back_03
  bone2Name=dyn_HairA_BackR_03
  swing1LimitDegrees=45
  swing1Motion=Locked
  swing2LimitDegrees=45
  swing2Motion=Locked
  twistLimitDegrees=45
  twistMotion=Locked
 }
 {
  bone1Name=attach_backpack
  bone2Name=spine_04
  swing1LimitDegrees=45
  swing1Motion=Free
  swing2LimitDegrees=45
  swing2Motion=Free
  twistLimitDegrees=45
  twistMotion=Free
 }
]
